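# Translucent, two-sided unlit lightning material for the HydroIgnite character's FX, usable on skeletal, spline and Niagara meshes and ribbons.
# Opacity sums two panned lightning textures and a smoke-trail glow, with vertex-color UV offsets, optional particle alpha and depth fade.
# Borrowed Parfait pickaxe noise drives vertex displacement, optionally adding SplineThicken width.
# Emissive blends base and oscillating accent colors, flickered by comparing vertex color against multiplied sines and computed per vertex.
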Material /BRCosmetics/Characters/Player/Male/Medium/Bodies/M_MED_HydroIgnite/FX/M_HydroIgnite_Lightning
graph {
 properties {
  bUsedWithMeshDeformer=true
  bUsedWithMorphTargets=true
  bUsedWithNiagaraMeshParticles=true
  bUsedWithNiagaraRibbons=true
  bUsedWithSkeletalMesh=true
  bUsedWithSplineMeshes=true
  blendMode=BLEND_Translucent
  pixelDepthOffsetMode=PDOM_Legacy
  shadingModel=MSM_Unlit
  twoSided=true
  usedShadingModels=Unlit
 }
 nodes [
  n1 = Time()
  n2 = Sine(None=Multiply(A=@n1) {constB=0.4})
  Clamp(None=@n2)
  Constant() {r=2}
  n3 = TextureCoordinate()
  n4 = MaterialFunctionCall(XY=@n3) {materialFunction=&Swizzle}
  n5 = ScalarParameter() {parameterName=Swizzle}
  n6 = VectorParameter() {group=Opacity, parameterName=SmokeTrail UV, sortPriority=1, defaultValue=<rgba 1 1 0.2 0>}
  n7 = VertexColor()
  n8 = Panner(
   Coordinate=Multiply(A=LinearInterpolate(A=@n3, B=@n4.YX, Alpha=@n5), B=AppendVector(A=@n6.R, B=@n6.R))
   Speed=Multiply(A=AppendVector(A=@n6.B, B=@n6.B), B=LinearInterpolate(Alpha=@n7) {constA=1, constB=1.1})
  )
  n9 = MaterialFunctionCall(Float2=@n8) {materialFunction=&BreakOutFloat2Components}
  n10 = LinearInterpolate(Alpha=VertexColor())
  n11 = LinearInterpolate(
   A=@n8
   B=MaterialFunctionCall(X=Add(A=@n9.R, B=@n10), Y=@n9.G) {materialFunction=&MakeFloat2}.Result
   Alpha=ScalarParameter() {defaultValue=1, parameterName=UseTextureOffset}
  ) {constAlpha=0}
  n12 = TextureSampleParameter2D(UVs=@n8) {group=Opacity, parameterName=SmokeTrailTexture, texture=/Game/Effects/Fort_Effects/Textures/Smoke/T_Smoke_Trail_Soft_Chromatic}
  n13 = Add(
   A=Add(
    A=TextureSample(UVs=@n11) {texture=/Game/Effects/Fort_Effects/Textures/T_Pickaxe_AllKnowingl_Lightning_01}.R
    B=TextureSample(UVs=@n11) {texture=/Game/Effects/Fort_Effects/Textures/T_Pickaxe_AllKnowingl_Lightning_02}.R
   )
   B=Multiply(A=@n12.R, B=ScalarParameter() {defaultValue=0.5, parameterName=GlowOpacity}) {constB=0.3}
  )
  n14 = ScalarParameter() {defaultValue=1, parameterName=Opacity}
  DepthFade(
   Opacity=LinearInterpolate(A=Multiply(A=@n13, B=@n14), B=ParticleColor().A, Alpha=ScalarParameter() {parameterName=UseParticleAlpha})
   FadeDistance=ScalarParameter() {defaultValue=5, parameterName=Depthfade}
  )
  DynamicParameter() {paramNames [NoiseIntensity, Param2, Param3, Param4]}
  n15 = MaterialFunctionCall!(
   Noise Intensity=ScalarParameter() {defaultValue=1, parameterName=NoiseIntensity}
   Noise UV Scale=ScalarParameter() {defaultValue=1, parameterName=Noise UV Scale}
   Time
   Noise Axis Rotation
  ) {materialFunction=/BRCosmetics/Weapons/FORT_Melee/Pickaxe_Parfait_Female/FX/Materials/Functions/MF_TrifleParfait_Noise}
  n16 = Add(A=LinearInterpolate(Alpha=VertexColor()) {constA=-0.5, constB=0}, B=ScalarParameter() {defaultValue=3, parameterName=SplineWidth})
  LinearInterpolate(
   A=Multiply(A=@n15.Noise WPO, B=Constant() {r=0.1})
   B=Add(
    A=@n15.Noise WPO
    B=MaterialFunctionCall(WidthBase=@n16, WidthTip=@n16) {materialFunction=&SplineThicken}.WorldPositionOffset
   )
   Alpha=ScalarParameter() {parameterName=Spline}
  )
  n17 = Sine(None=Multiply(A=@n1) {constB=0.7})
  n18 = If(A=VertexColor(), B=Clamp(None=Multiply(A=@n2, B=@n17)) {minDefault=-1}, "A > B"=Constant() {r=0.7}, "A < B"=Constant() {r=3})
  n19 = VectorParameter() {parameterName=Base Color, defaultValue=<rgba 0 0.2773 1 1>}
  n20 = Max(A=@n19.RGB, B=@n12.B)
  n21 = ScalarParameter() {defaultValue=3, parameterName=BaseIntensity}
  n22 = MaterialFunctionCall!(In) {materialFunction=/BRCosmetics/Weapons/FORT_Melee/Pickaxe_Parfait_Female/FX/Materials/Functions/MF_TrifleParfait_BrightnessOscillation}
  n23 = Multiply(A=Multiply(A=@n15.Noise Width Multiplier, B=@n22.Brightness), B=@n22.Brightness)
  n24 = ScalarParameter() {defaultValue=2, parameterName=Brightness Oscillation Exponent}
  n25 = LinearInterpolate(
   A=Multiply(A=Multiply(A=@n20, B=@n21) {constB=5}, B=@n19.RGB)
   B=Multiply(A=VectorParameter() {parameterName=AccentColor, defaultValue=<rgba 0 0.991 1 1>}.RGB, B=Power(Base=@n23, Exp=@n24))
  )
  MaterialFunctionCall(Emissive=VertexInterpolator(VS=Multiply(A=@n18, B=@n25)).PS) {materialFunction=&MF_FX_EmissiveAdjustments}
  ParticleColor()
 ]
}
