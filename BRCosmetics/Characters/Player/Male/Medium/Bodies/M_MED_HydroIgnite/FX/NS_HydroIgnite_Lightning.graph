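# Niagara system for the HydroIgnite character cosmetic: arcing lightning between body sockets, plus socket-attached glow sprites and flickering lights.
# Left, right and center start/end emitters seed socket-following points that looping ribbon emitters sample via Particle Attribute Readers, bending bursts into Hermite-spline bolts.
# A pickaxe-equipped flag increases beam counts and enables the center bolts and light intensity.
# Hand-hidden state controls glow visibility and lights; time of day scales their brightness.
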NiagaraSystem /BRCosmetics/Characters/Player/Male/Medium/Bodies/M_MED_HydroIgnite/FX/NS_HydroIgnite_Lightning
dependencies {
 usedDataInterfaces [&NiagaraDataInterfaceCurve, &NiagaraDataInterfaceArrayFloat3, &NiagaraDataInterfaceParticleRead]
 usedDynamicInputs [
  &AddVectorToPosition
  &Lerp_Vector
  &FN_FrontEndFloatModifier
  &SetBoolByFloatComparison
  &Boolean_AndOperation
 ]
 usedModules [
  &self:SystemSpawnScript.SystemScriptSource.SystemScriptGraph.NiagaraNodeAssignment_0.SetVariables_#1
  &SystemState
  &self:SystemSpawnScript.SystemScriptSource.SystemScriptGraph.NiagaraNodeAssignment_1.SetVariables_#2
  &Timeline
 ]
}
emitters {
 Glow {
  emitterUpdateScript [
   &EmitterState(Life Cycle Mode=::NewEnumerator1, Loop Behavior=::NewEnumerator1, Loop Duration Mode=::NewEnumerator1, UseLoopDelay=false)
   &SpawnBurst_Instantaneous(Spawn Count=2, Spawn Probability=HlslExpression() {hlslExpression=1.0}, Use Spawn Probability=false, UseLoopCountLimit=false)
  ]
  particleSpawnScript [
   &PartitionParticles()
   &InitializeParticle(Color=<rgba 0 4.0388 5 0.05>, Sprite Size Mode=::NewEnumerator3, Uniform Sprite Size=40, UsePositionOffset=false, Write Material Random=true)
  ]
  particleUpdateScript [
   &ParticleState(Kill Particles When Lifetime Has Elapsed=false, Let Infinitely Lived Particles Die When Emitter Deactivates=true)
   &SkeletalMeshLocation(
    Bone / Socket Index=$Particles.PartitionParticles.Partition
    MaskSpawnGroup=false
    Mesh Sampling Type=::NewEnumerator5
    Skeletal Mesh=$User.Skeletal Mesh
    Socket Sampling Mode=::NewEnumerator5
   )
   &ScaleColor(
    Scale Alpha=&Lerp_Float(
     A=0
     B=&FNTimeOfDayScalar(Night Scale=0.3, Morning Scale=0.5, Day Scale=0.6, Evening Scale=0.5)
     Alpha=&FloatFromCurve(
      FloatCurve=DataInterface() {propertyValues {Curve {keys [
          {interpMode=RCIM_Cubic, tangentMode=RCTM_User, time=0, value=0}
          {arriveTangent=2.4173, interpMode=RCIM_Cubic, leaveTangent=2.4173, tangentMode=RCTM_User, time=1, value=1}
         ]}}}
      CurveIndex=$System.Timeline.NormalizedTime
     )
    )
    ScaleA=true
    ScaleRGB=false
   )
   &CameraOffset(Camera Offset Amount=15)
   /SetVariables(Particles.VisibilityTag=&MakeCustomIntFromBool(Bool=$User.bisHandHidden, True INT=1, False INT=0)) {isSetParameters=true}
  ]
  rendererClasses [&NiagaraSpriteRendererProperties]
  renderers [{rendererClass=&NiagaraSpriteRendererProperties}]
 }
 Hermite_Curve_Ribbon001 {
  emitterUpdateScript [
   &EmitterState(
    Life Cycle Mode=::NewEnumerator1
    Loop Duration=&RandomRangeFloat(Minimum=0.5, Maximum=0.8)
    UseLoopDelay=false
   )
   /SetVariables(
    Emitter.NumofBeams:NiagaraInt32=&MakeCustomIntFromBool(Bool=$System.PickaxeEQ, True INT=2, False INT=1)
    Emitter.NumofBurst:NiagaraInt32=&MakeCustomIntFromBool(Bool=$System.PickaxeEQ, True INT=3, False INT=2)
   ) {isSetParameters=true, module=SetVariables_#3}
   /Game/Effects/Niagara/Modules/Partition/FN_PartitionAttributes(
    Number of Bone Sets=2
    Particles Per Partition=10
    Partitions:NiagaraInt32=&Multiply_Int(A=$Emitter.NumofBeams, B=$Emitter.NumofBurst)
   )
   &SpawnBurst_Instantaneous(
    Spawn Count=/Niagara/DynamicInputs/Divide/Divide_Int(A:NiagaraInt32=$Emitter.FN_PartitionAttributes.Spawn Total, B:NiagaraInt32=$Emitter.NumofBurst)
    Use Spawn Probability=false
   )
   &SpawnBurst_Instantaneous(
    Spawn Count=&Divide_Int(A:NiagaraInt32=$Emitter.FN_PartitionAttributes.Spawn Total, B:NiagaraInt32=$Emitter.NumofBurst)
    Spawn Time=&RandomRangeFloat(Minimum=0.05, Maximum=0.2)
    Use Spawn Probability=false
   ) {module=SpawnBurst_Instantaneous001}
   &SpawnBurst_Instantaneous(
    Spawn Count=&Divide_Int(A:NiagaraInt32=$Emitter.FN_PartitionAttributes.Spawn Total, B:NiagaraInt32=$Emitter.NumofBurst)
    Spawn Time=&RandomRangeFloat(Minimum=0.05, Maximum=0.2)
    Use Spawn Probability=false
   ) {module=SpawnBurst_Instantaneous002}
   /SetVariables(Emitter.Offset:NiagaraInt32=&RandomRangeInt(Minimum=0, Maximum=10)) {isSetParameters=true, module=SetVariables_#4}
  ]
  particleSpawnScript [
   /Game/Effects/Niagara/Modules/Partition/FN_Particle_partition_v5/FN_Particle_Partition_v5(Enable Ribbon ID=true) {enabled=false}
   &PartitionParticles(
    Normalized Index In Partition Includes One=true
    Number Of Partitions=$Emitter.Partitions
    Particles Per Partition:NiagaraInt32=$Emitter.ParticlesPerPartition
    Partition Mode:ENiagara_PartitionMode=::NewEnumerator1
    Write Partition Index to Ribbon Link Order:ENiagara_AttributeSamplingApplyOutput=::NewEnumerator0
    Write Partition to Ribbon ID=::NewEnumerator0
   )
   &InitializeRibbon(
    Lifetime=&RandomRangeFloat(Minimum=0.1, Maximum=0.2)
    Module.Write Color=true
    Module.Write Lifetime=true
    Module.Write Mass=true
    Module.Write Material Random=true
    Module.Write Position=false
    Module.Write RibbonTwist=false
    Module.Write RibbonWidth=true
    Ribbon Width=&RandomRangeFloat(Minimum=1.3, Maximum=1.7)
   )
   /SetVariables() {enabled=false, isSetParameters=true, module=SetVariables_#5}
  ]
  particleUpdateScript [
   &ParticleState()
   &self:LightningAttr(
    Offset:NiagaraInt32=$Emitter.Offset
    Particle Attribute Reader=DataInterface() {propertyValues {EmitterBinding {bindingMode=Other, emitterName=LightningStartR}}}
   ) {module=LightningAttr}
   &self:LightningAttrEnd(
    Offset:NiagaraInt32=$Emitter.Offset
    Particle Attribute Reader=DataInterface() {propertyValues {EmitterBinding {bindingMode=Other, emitterName=LightningEndR}}}
   ) {module=LightningAttrEnd}
   /Game/Effects/Niagara/Modules/Curves/FN_HermiteCubicSpline_v1/FN_HermiteCubicSpline(
    Compute Direction=true
    Direction End:Vector3f=&Multiply_VectorByFloat(Vector=$Particles.NVectorEnd, Float=&RandomRangeFloat(Minimum=30, Maximum=50))
    Direction Start:Vector3f=&Multiply_VectorByFloat(Vector=$Particles.NVectorStart, Float=&RandomRangeFloat(Minimum=30, Maximum=30))
    Position Blend:NiagaraFloat=$Particles.Initial.PartitionParticles.Normalized Index In Partition
    Position End:NiagaraPosition=$Particles.PositionEnd
    Position Start:NiagaraPosition=$Particles.PositionStart
    Write Particles Position=true
   )
   &FN_RibbonMaterial()
   &CameraOffset(Camera Offset Amount=-10) {enabled=false}
   &ScaleRibbonWidth(Ribbon Width Scale=&FloatFromCurve(
     FloatCurve=DataInterface() {propertyValues {Curve {keys [<at 0 1.25>,<at 0.8 1.2>,<at 1 0.9>]}}}
     CurveIndex=$Particles.Partition.NormalizedID
    )) {enabled=false}
   &DynamicMaterialParameters(Module.Index0Param0Write=true, Module.Index0Param1Write=true, Module.Index0Param2Write=true, Module.Index0Param3Write=true)
   /SetVariables() {enabled=false, isSetParameters=true, module=SetVariables_#6}
  ]
  rendererClasses [&NiagaraRibbonRendererProperties]
  renderers [{rendererClass=&NiagaraRibbonRendererProperties}]
 }
 Hermite_Curve_Ribbon003 {
  emitterUpdateScript [
   &EmitterState(
    Life Cycle Mode=::NewEnumerator1
    Loop Delay=0.3
    Loop Duration=&RandomRangeFloat(Minimum=0.5, Maximum=0.8)
    UseLoopDelay=true
   )
   /SetVariables(
    Emitter.NumofBeams:NiagaraInt32=&MakeCustomIntFromBool(Bool=$System.PickaxeEQ, True INT=2, False INT=1)
    Emitter.NumofBurst:NiagaraInt32=&MakeCustomIntFromBool(Bool=$System.PickaxeEQ, True INT=3, False INT=2)
   ) {isSetParameters=true, module=SetVariables_#3}
   &FN_PartitionAttributes(
    Number of Bone Sets=2
    Particles Per Partition=10
    Partitions:NiagaraInt32=&Multiply_Int(A=$Emitter.NumofBeams, B=$Emitter.NumofBurst)
   )
   &SpawnBurst_Instantaneous(
    Spawn Count=&Divide_Int(A:NiagaraInt32=$Emitter.FN_PartitionAttributes.Spawn Total, B:NiagaraInt32=$Emitter.NumofBurst)
    Use Spawn Probability=false
   )
   &SpawnBurst_Instantaneous(
    Spawn Count=&Divide_Int(A:NiagaraInt32=$Emitter.FN_PartitionAttributes.Spawn Total, B:NiagaraInt32=$Emitter.NumofBurst)
    Spawn Time=&RandomRangeFloat(Minimum=0.05, Maximum=0.2)
    Use Spawn Probability=false
   ) {module=SpawnBurst_Instantaneous001}
   &SpawnBurst_Instantaneous(
    Spawn Count=&Divide_Int(A:NiagaraInt32=$Emitter.FN_PartitionAttributes.Spawn Total, B:NiagaraInt32=$Emitter.NumofBurst)
    Spawn Time=&RandomRangeFloat(Minimum=0.1, Maximum=0.4)
    Use Spawn Probability=false
   ) {module=SpawnBurst_Instantaneous002}
   /SetVariables(Emitter.Offset:NiagaraInt32=&RandomRangeInt(Minimum=0, Maximum=10)) {isSetParameters=true, module=SetVariables_#4}
  ]
  particleSpawnScript [
   &FN_Particle_Partition_v5(Enable Ribbon ID=true) {enabled=false}
   &PartitionParticles(
    Normalized Index In Partition Includes One=true
    Number Of Partitions=$Emitter.Partitions
    Particles Per Partition:NiagaraInt32=$Emitter.ParticlesPerPartition
    Partition Mode:ENiagara_PartitionMode=::NewEnumerator1
    Write Partition Index to Ribbon Link Order:ENiagara_AttributeSamplingApplyOutput=::NewEnumerator0
    Write Partition to Ribbon ID=::NewEnumerator0
   )
   &InitializeRibbon(
    Lifetime=&RandomRangeFloat(Minimum=0.1, Maximum=0.2)
    Module.Write Color=true
    Module.Write Lifetime=true
    Module.Write Mass=true
    Module.Write Material Random=true
    Module.Write Position=false
    Module.Write RibbonTwist=false
    Module.Write RibbonWidth=true
    Ribbon Width=&RandomRangeFloat(Minimum=1.3, Maximum=1.7)
   )
   /SetVariables() {enabled=false, isSetParameters=true, module=SetVariables_#5}
  ]
  particleUpdateScript [
   &ParticleState()
   &self:LightningAttr(
    Offset:NiagaraInt32=$Emitter.Offset
    Particle Attribute Reader=DataInterface() {propertyValues {EmitterBinding {bindingMode=Other, emitterName=LightningStartL}}}
   ) {module=LightningAttr}
   &self:LightningAttrEnd(
    Offset:NiagaraInt32=$Emitter.Offset
    Particle Attribute Reader=DataInterface() {propertyValues {EmitterBinding {bindingMode=Other, emitterName=LightningEndL}}}
   ) {module=LightningAttrEnd}
   &FN_HermiteCubicSpline(
    Compute Direction=true
    Direction End:Vector3f=&Multiply_VectorByFloat(Vector=$Particles.NVectorEnd, Float=&RandomRangeFloat(Minimum=50, Maximum=50))
    Direction Start:Vector3f=&Multiply_VectorByFloat(Vector=$Particles.NVectorStart, Float=&RandomRangeFloat(Minimum=50, Maximum=50))
    Position Blend:NiagaraFloat=$Particles.Initial.PartitionParticles.Normalized Index In Partition
    Position End:NiagaraPosition=$Particles.PositionEnd
    Position Start:NiagaraPosition=$Particles.PositionStart
    Write Particles Position=true
   )
   &FN_RibbonMaterial()
   &CameraOffset(Camera Offset Amount=-10) {enabled=false}
   &ScaleRibbonWidth(Ribbon Width Scale=&FloatFromCurve(
     FloatCurve=DataInterface() {propertyValues {Curve {keys [<at 0 1.25>,<at 0.8 1.2>,<at 1 0.9>]}}}
     CurveIndex=$Particles.Partition.NormalizedID
    )) {enabled=false}
   &DynamicMaterialParameters(Module.Index0Param0Write=true, Module.Index0Param1Write=true, Module.Index0Param2Write=true, Module.Index0Param3Write=true)
   /SetVariables() {enabled=false, isSetParameters=true, module=SetVariables_#6}
  ]
  rendererClasses [&NiagaraRibbonRendererProperties]
  renderers [{rendererClass=&NiagaraRibbonRendererProperties}]
 }
 Hermite_Curve_Ribbon004 {
  emitterUpdateScript [
   &EmitterState(
    Life Cycle Mode=::NewEnumerator1
    Loop Delay=0.6
    Loop Duration=&RandomRangeFloat(Minimum=0.5, Maximum=0.8)
    UseLoopDelay=true
   )
   /SetVariables(
    Emitter.NumofBeams:NiagaraInt32=&MakeCustomIntFromBool(Bool=$System.PickaxeEQ, True INT=2, False INT=0)
    Emitter.NumofBurst:NiagaraInt32=&MakeCustomIntFromBool(Bool=$System.PickaxeEQ, True INT=3, False INT=2)
   ) {isSetParameters=true, module=SetVariables_#3}
   &FN_PartitionAttributes(
    Number of Bone Sets=2
    Particles Per Partition=10
    Partitions:NiagaraInt32=&Multiply_Int(A=$Emitter.NumofBeams, B=$Emitter.NumofBurst)
   )
   &SpawnBurst_Instantaneous(
    Spawn Count=&Divide_Int(A:NiagaraInt32=$Emitter.FN_PartitionAttributes.Spawn Total, B:NiagaraInt32=$Emitter.NumofBurst)
    Use Spawn Probability=false
   )
   &SpawnBurst_Instantaneous(
    Spawn Count=&Divide_Int(A:NiagaraInt32=$Emitter.FN_PartitionAttributes.Spawn Total, B:NiagaraInt32=$Emitter.NumofBurst)
    Spawn Time=&RandomRangeFloat(Minimum=0.1, Maximum=0.2)
    Use Spawn Probability=false
   ) {module=SpawnBurst_Instantaneous001}
   &SpawnBurst_Instantaneous(
    Spawn Count=&Divide_Int(A:NiagaraInt32=$Emitter.FN_PartitionAttributes.Spawn Total, B:NiagaraInt32=$Emitter.NumofBurst)
    Spawn Time=&RandomRangeFloat(Minimum=0.1, Maximum=0.2)
    Use Spawn Probability=false
   ) {module=SpawnBurst_Instantaneous002}
   /SetVariables(Emitter.Offset:NiagaraInt32=&RandomRangeInt(Minimum=0, Maximum=10)) {isSetParameters=true, module=SetVariables_#4}
  ]
  particleSpawnScript [
   &FN_Particle_Partition_v5(Enable Ribbon ID=true) {enabled=false}
   &PartitionParticles(
    Normalized Index In Partition Includes One=true
    Number Of Partitions=$Emitter.Partitions
    Particles Per Partition:NiagaraInt32=$Emitter.ParticlesPerPartition
    Partition Mode:ENiagara_PartitionMode=::NewEnumerator1
    Write Partition Index to Ribbon Link Order:ENiagara_AttributeSamplingApplyOutput=::NewEnumerator0
    Write Partition to Ribbon ID=::NewEnumerator0
   )
   &InitializeRibbon(
    Lifetime=&RandomRangeFloat(Minimum=0.1, Maximum=0.2)
    Module.Write Color=true
    Module.Write Lifetime=true
    Module.Write Mass=true
    Module.Write Material Random=true
    Module.Write Position=false
    Module.Write RibbonTwist=false
    Module.Write RibbonWidth=true
    Ribbon Width=&RandomRangeFloat(Minimum=2.5, Maximum=3)
   )
   /SetVariables() {enabled=false, isSetParameters=true, module=SetVariables_#5}
  ]
  particleUpdateScript [
   &ParticleState()
   &self:LightningAttr(
    Offset:NiagaraInt32=$Emitter.Offset
    Particle Attribute Reader=DataInterface() {propertyValues {EmitterBinding {bindingMode=Other, emitterName=LightningStartC}}}
   ) {module=LightningAttr}
   &self:LightningAttrEnd(
    Offset:NiagaraInt32=$Emitter.Offset
    Particle Attribute Reader=DataInterface() {propertyValues {EmitterBinding {bindingMode=Other, emitterName=LightningEndC}}}
   ) {module=LightningAttrEnd}
   &FN_HermiteCubicSpline(
    Compute Direction=true
    Direction End:Vector3f=&Multiply_VectorByFloat(Vector=$Particles.NVectorEnd, Float=&RandomRangeFloat(Minimum=50, Maximum=50))
    Direction Start:Vector3f=&Multiply_VectorByFloat(Vector=$Particles.NVectorStart, Float=&RandomRangeFloat(Minimum=50, Maximum=50))
    Position Blend:NiagaraFloat=$Particles.Initial.PartitionParticles.Normalized Index In Partition
    Position End:NiagaraPosition=$Particles.PositionEnd
    Position Start:NiagaraPosition=$Particles.PositionStart
    Write Particles Position=true
   )
   &FN_RibbonMaterial()
   &CameraOffset(Camera Offset Amount=-10) {enabled=false}
   &ScaleRibbonWidth(Ribbon Width Scale=&FloatFromCurve(
     FloatCurve=DataInterface() {propertyValues {Curve {keys [<at 0 1.25>,<at 0.8 1.2>,<at 1 0.9>]}}}
     CurveIndex=$Particles.Partition.NormalizedID
    )) {enabled=false}
   &DynamicMaterialParameters(Module.Index0Param0Write=true, Module.Index0Param1Write=true, Module.Index0Param2Write=true, Module.Index0Param3Write=true)
   /SetVariables() {enabled=false, isSetParameters=true, module=SetVariables_#6}
  ]
  rendererClasses [&NiagaraRibbonRendererProperties]
  renderers [{rendererClass=&NiagaraRibbonRendererProperties}]
 }
 Light {
  emitterSpawnScript [&FN_PreserveSkydiveUserVariables() {enabled=false, module=PreserveUserVariables}]
  emitterUpdateScript [
   &EmitterState()
   &SpawnBurst_Instantaneous(Spawn Count=2, Spawn Time=0.01, Use Spawn Probability=false, UseLoopCountLimit=true)
  ]
  particleSpawnScript [
   &PartitionParticles()
   &InitializeParticle(Sprite Size Mode=::NewEnumerator3, UsePositionOffset=false, Write Material Random=true)
  ]
  particleUpdateScript [
   &ParticleState(
    Kill Particles When Lifetime Has Elapsed=false
    Let Infinitely Lived Particles Die When Emitter Deactivates=true
    Loop Particles Lifetime=true
   )
   &SkeletalMeshLocation(
    Bone / Socket Index=$Particles.PartitionParticles.Partition
    MaskSpawnGroup=false
    Mesh Sampling Type=::NewEnumerator5
    Skeletal Mesh=$User.Skeletal Mesh
    Socket Sampling Mode=::NewEnumerator5
   )
   &FN_Light_Attributes(
    Color=<rgba 0 0.807 1 1>
    Create Color Attribute=true
    Create Exponent Attribute=true
    Create Radius Attribute=true
    Enable Particle Light Enable=true
    Exponent=2
    Light Enabled:NiagaraBool=&Boolean_NotOperation(Bool=$User.bisHandHidden)
    Radius=20
    Scale Color=&Multiply_Float(
     A=&Multiply_Float(
      A=&FNTimeOfDayScalar(Night Scale=2.5, Morning Scale=3, Day Scale=4, Evening Scale=3)
      B=&MakeFloatFromInt(Int=&MakeCustomIntFromBool(Bool=$System.PickaxeEQ, True INT=2, False INT=0))
     )
     B=&Waveform(
      Waveform Count=::NewEnumerator2
      "Waveform [1]"=::NewEnumerator0
      "[1] Frequency"=1
      "[1] Amplitude Min/Max"=<0.9 1>
      "[1] Scale"=0.5
      "[2] Average"=false
      "Waveform [2]"=::NewEnumerator5
      "[2] Input Position"=0
      "[2] Phase"=$Emitter.Age
      "[2] Phase Scale"=1
      "[2] Compound Frequency"=2
      "[2] Frequency"=2
      "[2] Amplitude Min/Max"=<0 1>
      "[2] Scale"=0.05
      "[2] Bias"=0
      "[2] Exponent"=1
      "[2] Pulse Duration"=0.3
      "[2] Clip Triangle Min/Max"=<0 0>
      "[2] Random Seed"=0
      "[3] Average"=false
      "Waveform [3]"=::NewEnumerator5
      "[3] Input Position"=0
      "[3] Phase"=$Emitter.Age
      "[3] Phase Scale"=1
      "[3] Frequency"=1
      "[3] Compound Frequency"=2
      "[3] Amplitude Min/Max"=<0 1>
      "[3] Scale"=0.03
      "[3] Bias"=0
      "[3] Exponent"=1
      "[3] Pulse Duration"=0.3
      "[3] Clip Triangle Min/Max"=<0 0>
      "[3] Random Seed"=0
     )
    )
   )
  ]
  rendererClasses [&NiagaraLightRendererProperties, &NiagaraSpriteRendererProperties]
  renderers [
   {rendererClass=&NiagaraLightRendererProperties, rendererIndex=0}
   {rendererClass=&NiagaraSpriteRendererProperties, rendererIndex=1}
  ]
 }
 LightningEndC {
  emitterUpdateScript [
   &EmitterState(Life Cycle Mode=::NewEnumerator1, UseLoopDelay=false)
   &SpawnBurst_Instantaneous(Spawn Count=25, Use Spawn Probability=false, UseLoopCountLimit=false)
  ]
  particleSpawnScript [
   &PartitionParticles(Number Of Partitions=4, Partition Mode:ENiagara_PartitionMode=::NewEnumerator5)
   &InitializeParticle(Lifetime=5, Sprite Size Mode=::NewEnumerator3, Uniform Sprite Size=1, UsePositionOffset=false, Write Material Random=true)
   &ShapeLocation(
    Hemisphere Distribution=<1 0.1>
    MaskSpawnGroup=false
    Override Seed=false
    Rotation Mode=::NewEnumerator2
    Sphere Radius=0.5
    Sphere Surface Distribution=0.5833
    UseOwnerScale=true
    Yaw / Pitch / Roll=&SelectVectorFromArray(
     Direct Array Index=$Particles.PartitionParticles.Partition
     Vector Selection Array=DataInterface() {propertyValues {FloatData [<0.5665 0 -0.1812>, <0.5665 0 -0.1812>, <-0.2474 0.2753 -0.164>, <0.6296 0 -0.1812>]}}
    )
   )
  ]
  particleUpdateScript [
   &ParticleState()
   &ScaleColor(
    Scale Alpha=&FloatFromCurve(FloatCurve=DataInterface() {propertyValues {Curve {keys [<at -0.0046 1.0113 RCIM_Cubic>,<at 1.0011 0.002 RCIM_Cubic>]}}})
    ScaleA=true
    ScaleRGB=false
   )
   &SkeletalMeshLocation(
    Bone / Socket Index=$Particles.PartitionParticles.Partition
    MaskSpawnGroup=false
    Mesh Sampling Type=::NewEnumerator5
    Orientation Sampling=::NewEnumerator1
    Sampled Position Offset=$Particles.Initial.ShapeLocation.ShapePosition
    Skeletal Mesh=DataInterface() {propertyValues {
      DefaultMesh=null
      ExcludeBoneName=root
      OptionalComponentClass=null
      PreviewMesh=/BRCosmetics/Characters/Player/Male/Medium/Bodies/M_MED_HydroIgnite/Meshes/M_MED_HydroIgnite
      SkinningMode=SkinOnTheFly
      SoftSourceActor=null
      SourceMode=Default
      UvSetIndex=0
      WholeMeshLOD=-1
      bExcludeBone=false
      bReadDeformedGeometry=true
      bRequireCurrentFrameData=true
      ComponentTags []
      FilteredBones []
      FilteredSockets [FX_EC1,FX_EC2,FX_SR4,FX_SL4]
      MeshUserParameter {parameter {name=None, typeDefHandle {registeredTypeIndex=1}, varData []}}
      SamplingRegions []
     }}
    Socket Sampling Mode=::NewEnumerator5
   )
   &LinearForce(Force=&Multiply_VectorByFloat(Vector=$Particles.ShapeLocation.ShapeVector, Float=103.4008)) {enabled=false}
   &SolveForcesAndVelocity(Clamp Velocity=false, Limit Acceleration=false)
  ]
  rendererClasses [&NiagaraSpriteRendererProperties]
  renderers [{rendererClass=&NiagaraSpriteRendererProperties}]
 }
 LightningEndL {
  emitterUpdateScript [
   &EmitterState(Life Cycle Mode=::NewEnumerator1, UseLoopDelay=false)
   &SpawnBurst_Instantaneous(Spawn Count=24, Use Spawn Probability=false, UseLoopCountLimit=false)
  ]
  particleSpawnScript [
   &PartitionParticles(Partition Mode:ENiagara_PartitionMode=::NewEnumerator5)
   &InitializeParticle(Lifetime=5, Sprite Size Mode=::NewEnumerator3, Uniform Sprite Size=1, UsePositionOffset=false, Write Material Random=true)
   &ShapeLocation(
    Hemisphere Distribution=<1 0.1>
    MaskSpawnGroup=false
    Override Seed=false
    Rotation Mode=::NewEnumerator2
    Sphere Radius=0.5
    Sphere Surface Distribution=0.5833
    UseOwnerScale=true
    Yaw / Pitch / Roll=&SelectVectorFromArray(
     Direct Array Index=$Particles.PartitionParticles.Partition
     Vector Selection Array=DataInterface() {propertyValues {FloatData [<0.8426 0.2064 0>, <0.8426 0.2064 0>, <0.1261 0.2981 0>]}}
    )
   )
  ]
  particleUpdateScript [
   &ParticleState()
   &ScaleColor(
    Scale Alpha=&FloatFromCurve(FloatCurve=DataInterface() {propertyValues {Curve {keys [<at -0.0046 1.0113 RCIM_Cubic>,<at 1.0011 0.002 RCIM_Cubic>]}}})
    ScaleA=true
    ScaleRGB=false
   )
   &SkeletalMeshLocation(
    Bone / Socket Index=$Particles.PartitionParticles.Partition
    MaskSpawnGroup=false
    Mesh Sampling Type=::NewEnumerator5
    Orientation Sampling=::NewEnumerator1
    Sampled Position Offset=$Particles.Initial.ShapeLocation.ShapePosition
    Skeletal Mesh=DataInterface() {propertyValues {
      DefaultMesh=null
      ExcludeBoneName=root
      OptionalComponentClass=null
      PreviewMesh=&M_MED_HydroIgnite
      SkinningMode=SkinOnTheFly
      SoftSourceActor=null
      SourceMode=Default
      UvSetIndex=0
      WholeMeshLOD=-1
      bExcludeBone=false
      bReadDeformedGeometry=true
      bRequireCurrentFrameData=true
      ComponentTags []
      FilteredBones []
      FilteredSockets [FX_EL1,FX_EL2]
      MeshUserParameter {parameter {name=None, typeDefHandle {registeredTypeIndex=1}, varData []}}
      SamplingRegions []
     }}
    Socket Sampling Mode=::NewEnumerator5
   )
   &LinearForce(Force=&Multiply_VectorByFloat(Vector=$Particles.ShapeLocation.ShapeVector, Float=103.4008)) {enabled=false}
   &SolveForcesAndVelocity(Clamp Velocity=false, Limit Acceleration=false)
  ]
  rendererClasses [&NiagaraSpriteRendererProperties]
  renderers [{rendererClass=&NiagaraSpriteRendererProperties}]
 }
 LightningEndR {
  emitterUpdateScript [
   &EmitterState(Life Cycle Mode=::NewEnumerator1, UseLoopDelay=false)
   &SpawnBurst_Instantaneous(Spawn Count=25, Use Spawn Probability=false, UseLoopCountLimit=false)
  ]
  particleSpawnScript [
   &PartitionParticles(Partition Mode:ENiagara_PartitionMode=::NewEnumerator5)
   &InitializeParticle(Lifetime=5, Sprite Size Mode=::NewEnumerator3, Uniform Sprite Size=1, UsePositionOffset=false, Write Material Random=true)
   &ShapeLocation(
    Hemisphere Distribution=<1 0.1>
    MaskSpawnGroup=false
    Override Seed=false
    Rotation Mode=::NewEnumerator2
    Sphere Radius=0.5
    Sphere Surface Distribution=0.5833
    UseOwnerScale=true
    Yaw / Pitch / Roll=&SelectVectorFromArray(
     Direct Array Index=$Particles.PartitionParticles.Partition
     Vector Selection Array=DataInterface() {propertyValues {FloatData [<0.9276 0.2064 0>, <0 0.4127 0.847>, <0.1261 0.2981 0>]}}
    )
   )
  ]
  particleUpdateScript [
   &ParticleState()
   &ScaleColor(
    Scale Alpha=&FloatFromCurve(FloatCurve=DataInterface() {propertyValues {Curve {keys [<at -0.0046 1.0113 RCIM_Cubic>,<at 1.0011 0.002 RCIM_Cubic>]}}})
    ScaleA=true
    ScaleRGB=false
   )
   &SkeletalMeshLocation(
    Bone / Socket Index=$Particles.PartitionParticles.Partition
    MaskSpawnGroup=false
    Mesh Sampling Type=::NewEnumerator5
    Orientation Sampling=::NewEnumerator1
    Sampled Position Offset=$Particles.Initial.ShapeLocation.ShapePosition
    Skeletal Mesh=DataInterface() {propertyValues {
      DefaultMesh=null
      ExcludeBoneName=root
      OptionalComponentClass=null
      PreviewMesh=&M_MED_HydroIgnite
      SkinningMode=SkinOnTheFly
      SoftSourceActor=null
      SourceMode=Default
      UvSetIndex=0
      WholeMeshLOD=-1
      bExcludeBone=false
      bReadDeformedGeometry=true
      bRequireCurrentFrameData=true
      ComponentTags []
      FilteredBones []
      FilteredSockets [FX_ER1,FX_ER2]
      MeshUserParameter {parameter {name=None, typeDefHandle {registeredTypeIndex=1}, varData []}}
      SamplingRegions []
     }}
    Socket Sampling Mode=::NewEnumerator5
   )
   &SolveForcesAndVelocity(Clamp Velocity=false, Limit Acceleration=false)
  ]
  rendererClasses [&NiagaraSpriteRendererProperties]
  renderers [{rendererClass=&NiagaraSpriteRendererProperties}]
 }
 LightningStartC {
  emitterUpdateScript [
   &EmitterState(Life Cycle Mode=::NewEnumerator1, UseLoopDelay=false)
   &SpawnBurst_Instantaneous(Spawn Count=25, Use Spawn Probability=false, UseLoopCountLimit=false)
  ]
  particleSpawnScript [
   &PartitionParticles(Partition Mode:ENiagara_PartitionMode=::NewEnumerator5)
   &InitializeParticle(Lifetime=5, Sprite Size Mode=::NewEnumerator3, Uniform Sprite Size=1, UsePositionOffset=false, Write Material Random=true)
   &ShapeLocation(
    Hemisphere Distribution=<1 0.1>
    MaskSpawnGroup=false
    Override Seed=false
    Rotation Mode=::NewEnumerator2
    Sphere Radius=0.5
    Sphere Surface Distribution=0.5833
    UseOwnerScale=true
    Yaw / Pitch / Roll=&SelectVectorFromArray(
     Direct Array Index=$Particles.PartitionParticles.Partition
     Vector Selection Array=DataInterface() {propertyValues {FloatData [<0.8254 0.2981 0>, <0.8254 0.2981 0>, <0 0.3038 0.0251>, <0 0.3038 0.0251>, <0 0.3038 0.0251>]}}
    )
   )
  ]
  particleUpdateScript [
   &ParticleState()
   &ScaleColor(
    Scale Alpha=&FloatFromCurve(FloatCurve=DataInterface() {propertyValues {Curve {keys [<at -0.0046 1.0113 RCIM_Cubic>,<at 1.0011 0.002 RCIM_Cubic>]}}})
    ScaleA=true
    ScaleRGB=false
   )
   &SkeletalMeshLocation(
    Bone / Socket Index=$Particles.PartitionParticles.Partition
    MaskSpawnGroup=false
    Mesh Sampling Type=::NewEnumerator5
    Orientation Sampling=::NewEnumerator1
    Sampled Position Offset=$Particles.Initial.ShapeLocation.ShapePosition
    Skeletal Mesh=DataInterface() {propertyValues {
      DefaultMesh=null
      ExcludeBoneName=root
      OptionalComponentClass=null
      PreviewMesh=&M_MED_HydroIgnite
      SkinningMode=SkinOnTheFly
      SoftSourceActor=null
      SourceMode=Default
      UvSetIndex=0
      WholeMeshLOD=-1
      bExcludeBone=false
      bReadDeformedGeometry=true
      bRequireCurrentFrameData=true
      ComponentTags []
      FilteredBones []
      FilteredSockets [FX_SC1,FX_SC2,None,None,None]
      MeshUserParameter {parameter {name=None, typeDefHandle {registeredTypeIndex=1}, varData []}}
      SamplingRegions []
     }}
    Socket Sampling Mode=::NewEnumerator5
   )
   &LinearForce(Force=&Multiply_VectorByFloat(Vector=$Particles.ShapeLocation.ShapeVector, Float=103.4008)) {enabled=false}
   &SolveForcesAndVelocity(Clamp Velocity=false, Limit Acceleration=false)
  ]
  rendererClasses [&NiagaraSpriteRendererProperties]
  renderers [{rendererClass=&NiagaraSpriteRendererProperties}]
 }
 LightningStartL {
  emitterUpdateScript [
   &EmitterState(Life Cycle Mode=::NewEnumerator1, UseLoopDelay=false)
   &SpawnBurst_Instantaneous(Spawn Count=25, Use Spawn Probability=false, UseLoopCountLimit=false)
  ]
  particleSpawnScript [
   &PartitionParticles(Number Of Partitions=3, Partition Mode:ENiagara_PartitionMode=::NewEnumerator5)
   &InitializeParticle(Lifetime=5, Sprite Size Mode=::NewEnumerator3, Uniform Sprite Size=1, UsePositionOffset=false, Write Material Random=true)
   &ShapeLocation(
    Hemisphere Distribution=<1 0.1>
    MaskSpawnGroup=false
    Override Seed=false
    Rotation Mode=::NewEnumerator2
    Sphere Radius=0.5
    Sphere Surface Distribution=0.5833
    UseOwnerScale=true
    Yaw / Pitch / Roll=&SelectVectorFromArray(
     Direct Array Index=$Particles.PartitionParticles.Partition
     Vector Selection Array=DataInterface() {propertyValues {FloatData [<0.6296 0 -0.1812>, <0 0.3267 0.1994>, <0.7839 0.6534 0.5038>, <0 0.3038 0.0251>, <0 0.3038 0.0251>]}}
    )
   )
  ]
  particleUpdateScript [
   &ParticleState()
   &ScaleColor(
    Scale Alpha=&FloatFromCurve(FloatCurve=DataInterface() {propertyValues {Curve {keys [<at -0.0046 1.0113 RCIM_Cubic>,<at 1.0011 0.002 RCIM_Cubic>]}}})
    ScaleA=true
    ScaleRGB=false
   )
   &SkeletalMeshLocation(
    Bone / Socket Index=$Particles.PartitionParticles.Partition
    MaskSpawnGroup=false
    Mesh Sampling Type=::NewEnumerator5
    Orientation Sampling=::NewEnumerator1
    Sampled Position Offset=$Particles.Initial.ShapeLocation.ShapePosition
    Skeletal Mesh=DataInterface() {propertyValues {
      DefaultMesh=null
      ExcludeBoneName=root
      OptionalComponentClass=null
      PreviewMesh=&M_MED_HydroIgnite
      SkinningMode=SkinOnTheFly
      SoftSourceActor=null
      SourceMode=Default
      UvSetIndex=0
      WholeMeshLOD=-1
      bExcludeBone=false
      bReadDeformedGeometry=true
      bRequireCurrentFrameData=true
      ComponentTags []
      FilteredBones []
      FilteredSockets [FX_SL4,FX_SL1,FX_SL3]
      MeshUserParameter {parameter {name=None, typeDefHandle {registeredTypeIndex=1}, varData []}}
      SamplingRegions []
     }}
    Socket Sampling Mode=::NewEnumerator5
   )
   &LinearForce(Force=&Multiply_VectorByFloat(Vector=$Particles.ShapeLocation.ShapeVector, Float=103.4008)) {enabled=false}
   &SolveForcesAndVelocity(Clamp Velocity=false, Limit Acceleration=false)
  ]
  rendererClasses [&NiagaraSpriteRendererProperties]
  renderers [{rendererClass=&NiagaraSpriteRendererProperties}]
 }
 LightningStartR {
  emitterUpdateScript [
   &EmitterState(Life Cycle Mode=::NewEnumerator1, UseLoopDelay=false)
   &SpawnBurst_Instantaneous(Spawn Count=25, Use Spawn Probability=false, UseLoopCountLimit=false)
  ]
  particleSpawnScript [
   &PartitionParticles(Number Of Partitions=3, Partition Mode:ENiagara_PartitionMode=::NewEnumerator5)
   &InitializeParticle(Lifetime=5, Sprite Size Mode=::NewEnumerator3, Uniform Sprite Size=1, UsePositionOffset=false, Write Material Random=true)
   &ShapeLocation(
    Hemisphere Distribution=<1 0.1>
    MaskSpawnGroup=false
    Override Seed=false
    Rotation Mode=::NewEnumerator2
    Sphere Radius=0.5
    Sphere Surface Distribution=0.5833
    UseOwnerScale=true
    Yaw / Pitch / Roll=&SelectVectorFromArray(
     Direct Array Index=$Particles.PartitionParticles.Partition
     Vector Selection Array=DataInterface() {propertyValues {FloatData [<-0.2474 0.2753 -0.164>, <0 0.4127 0.847>, <-0.0185 0.6534 0.5038>, <0 0.3038 0.0251>, <0 0.3038 0.0251>]}}
    )
   )
  ]
  particleUpdateScript [
   &ParticleState()
   &ScaleColor(
    Scale Alpha=&FloatFromCurve(FloatCurve=DataInterface() {propertyValues {Curve {keys [<at -0.0046 1.0113 RCIM_Cubic>,<at 1.0011 0.002 RCIM_Cubic>]}}})
    ScaleA=true
    ScaleRGB=false
   )
   &SkeletalMeshLocation(
    Bone / Socket Index=$Particles.PartitionParticles.Partition
    MaskSpawnGroup=false
    Mesh Sampling Type=::NewEnumerator5
    Orientation Sampling=::NewEnumerator1
    Sampled Position Offset=$Particles.Initial.ShapeLocation.ShapePosition
    Skeletal Mesh=DataInterface() {propertyValues {
      DefaultMesh=null
      ExcludeBoneName=root
      OptionalComponentClass=null
      PreviewMesh=&M_MED_HydroIgnite
      SkinningMode=SkinOnTheFly
      SoftSourceActor=null
      SourceMode=Default
      UvSetIndex=0
      WholeMeshLOD=-1
      bExcludeBone=false
      bReadDeformedGeometry=true
      bRequireCurrentFrameData=true
      ComponentTags []
      FilteredBones []
      FilteredSockets [FX_SR4,FX_SR3,FX_SR1]
      MeshUserParameter {parameter {name=None, typeDefHandle {registeredTypeIndex=1}, varData []}}
      SamplingRegions []
     }}
    Socket Sampling Mode=::NewEnumerator5
   )
   &SolveForcesAndVelocity(Clamp Velocity=false, Limit Acceleration=false)
  ]
  rendererClasses [&NiagaraSpriteRendererProperties]
  renderers [{rendererClass=&NiagaraSpriteRendererProperties}]
 }
}
summary {
 systemName=NS_HydroIgnite_Lightning
 emitters [
  {emitterName=Glow, rendererClasses [&NiagaraSpriteRendererProperties]}
  {emitterName=LightningStartR, rendererClasses [&NiagaraSpriteRendererProperties]}
  {emitterName=Hermite_Curve_Ribbon001, rendererClasses [&NiagaraRibbonRendererProperties]}
  {emitterName=LightningEndR, rendererClasses [&NiagaraSpriteRendererProperties]}
  {emitterName=LightningStartL, rendererClasses [&NiagaraSpriteRendererProperties]}
  {emitterName=LightningEndL, rendererClasses [&NiagaraSpriteRendererProperties]}
  {emitterName=Hermite_Curve_Ribbon003, rendererClasses [&NiagaraRibbonRendererProperties]}
  {emitterName=LightningStartC, rendererClasses [&NiagaraSpriteRendererProperties]}
  {emitterName=LightningEndC, rendererClasses [&NiagaraSpriteRendererProperties]}
  {emitterName=Hermite_Curve_Ribbon004, rendererClasses [&NiagaraRibbonRendererProperties]}
  {emitterName=Light, rendererClasses [&NiagaraLightRendererProperties, &NiagaraSpriteRendererProperties]}
 ]
 userVariables [
  User.bIsFrontEnd {type=NiagaraBool}
  User.bIsFrontEndPreview {type=NiagaraBool}
  User.bisHandHidden {type=NiagaraBool}
  User.Pickaxe {type=NiagaraFloat, defaultValue=0, description=""}
  User.Skeletal Mesh {
   type=NiagaraDataInterfaceSkeletalMesh
   defaultValue {
    struct=NiagaraExt_VariableValue_DataInterface
    value {dataInterface=&self:NiagaraDataInterfaceSkeletalMesh_0}
   }
  }
 ]
}
