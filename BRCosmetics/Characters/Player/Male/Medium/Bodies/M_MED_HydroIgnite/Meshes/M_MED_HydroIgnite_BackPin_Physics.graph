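# Physics asset for the BackPin mesh of the male medium HydroIgnite cosmetic body, defining secondary motion for its vest flaps and straps.
# Capsule chains form three back panels, two front panels and an upper strap, each rooted by a MSTR body constrained to spine_01.
# The lower strap branches from the front-left vest chain.
# Back panels swing mainly on one axis, while larger pelvis, spine and thigh capsules with free angular constraints represent the body.
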PhysicsAsset /BRCosmetics/Characters/Player/Male/Medium/Bodies/M_MED_HydroIgnite/Meshes/M_MED_HydroIgnite_BackPin_Physics
bodies {
 dyn_LowerStrap_01_jnt [{shapeType=Capsule, length=1.5, radius=0.25, shapeName=dyn_LowerStrap_01_jnt_capsule, center=<1 0 0>, rotation=<rot 0 0 45>}]
 dyn_LowerStrap_02_jnt [{shapeType=Capsule, length=1.5, radius=0.25, shapeName=dyn_LowerStrap_02_jnt_capsule, center=<1 0 0>, rotation=<rot 0 0 45>}]
 dyn_LowerStrap_03_jnt [{shapeType=Capsule, length=1.5, radius=0.25, shapeName=dyn_LowerStrap_03_jnt_capsule, center=<2 0 0>, rotation=<rot 0 0 45>}]
 dyn_LowerStrap_MSTR [{shapeType=Capsule, length=0.505, radius=0.505, shapeName=dyn_LowerStrap_MSTR_capsule, rotation=<rot 0 -90 0>}]
 dyn_UpperStrap_01_jnt [{shapeType=Capsule, length=1, radius=1, shapeName=dyn_UpperStrap_01_jnt_capsule, center=<1.5 0 0>, rotation=<rot 90 0 0>}]
 dyn_UpperStrap_02_jnt [{shapeType=Capsule, length=1, radius=1, shapeName=dyn_UpperStrap_02_jnt_capsule, center=<1.5 0 0>, rotation=<rot 90 0 0>}]
 dyn_UpperStrap_03_jnt [{shapeType=Capsule, length=1, radius=1, shapeName=dyn_UpperStrap_03_jnt_capsule, center=<1.5 0 0>, rotation=<rot 90 0 0>}]
 dyn_UpperStrap_MSTR [{shapeType=Capsule, length=0.505, radius=0.505, shapeName=dyn_UpperStrap_MSTR_capsule, rotation=<rot 0 -90 0>}]
 dyn_bk_Vest_01_jnt [{shapeType=Capsule, length=5, radius=1.5, shapeName=dyn_bk_Vest_01_jnt_capsule, center=<5 -0.5 0>, rotation=<rot 90 0 0>}]
 dyn_bk_Vest_02_jnt [{shapeType=Capsule, length=5, radius=1.5, shapeName=dyn_bk_Vest_02_jnt_capsule, center=<5 -0.5 0>, rotation=<rot 90 0 0>}]
 dyn_bk_Vest_MSTR [{shapeType=Capsule, length=0.505, radius=0.505, shapeName=dyn_bk_Vest_MSTR_capsule, rotation=<rot 0 -90 0>}]
 dyn_bk_l_Vest_01_jnt [{shapeType=Capsule, length=5, radius=1.5, shapeName=dyn_bk_l_Vest_01_jnt_capsule, center=<5 0 0>, rotation=<rot 90 0 0>}]
 dyn_bk_l_Vest_02_jnt [{shapeType=Capsule, length=5, radius=1.5, shapeName=dyn_bk_l_Vest_02_jnt_capsule, center=<5 0 0>, rotation=<rot 90 0 0>}]
 dyn_bk_l_Vest_MSTR [{shapeType=Capsule, length=0.505, radius=0.505, shapeName=dyn_bk_l_Vest_MSTR_capsule, rotation=<rot 0 -90 0>}]
 dyn_bk_r_Vest_01_jnt [{shapeType=Capsule, length=5, radius=1.5, shapeName=dyn_bk_r_Vest_01_jnt_capsule, center=<5 0 0>, rotation=<rot 90 0 0>}]
 dyn_bk_r_Vest_02_jnt [{shapeType=Capsule, length=5, radius=1.5, shapeName=dyn_bk_r_Vest_02_jnt_capsule, center=<5 0 0>, rotation=<rot 90 0 0>}]
 dyn_bk_r_Vest_MSTR [{shapeType=Capsule, length=0.505, radius=0.505, shapeName=dyn_bk_r_Vest_MSTR_capsule, rotation=<rot 0 -90 0>}]
 dyn_ft_l_Vest_01_jnt [{shapeType=Capsule, length=10, radius=2, shapeName=dyn_ft_l_Vest_01_jnt_capsule, center=<4 -2.5 0>, rotation=<rot 90 0 90>}]
 dyn_ft_l_Vest_02_jnt [{shapeType=Capsule, length=5, radius=1.75, shapeName=dyn_ft_l_Vest_02_jnt_capsule, center=<10 3 0.5>, rotation=<rot 90 0 0>}]
 dyn_ft_l_Vest_03_jnt [{shapeType=Capsule, length=13, radius=1.5, shapeName=dyn_ft_l_Vest_03_jnt_capsule, center=<7 -2.5 0>, rotation=<rot 85 -90 35>}]
 dyn_ft_l_Vest_MSTR [{shapeType=Capsule, length=0.505, radius=0.505, shapeName=dyn_ft_l_Vest_MSTR_capsule, rotation=<rot 0 -90 0>}]
 dyn_ft_r_Vest_01_jnt [{shapeType=Capsule, length=10, radius=2, shapeName=dyn_ft_r_Vest_01_jnt_capsule, center=<4 2.5 0>, rotation=<rot 90 0 90>}]
 dyn_ft_r_Vest_02_jnt [{shapeType=Capsule, length=15, radius=1.5, shapeName=dyn_ft_r_Vest_02_jnt_capsule, center=<8 -3 0.5>, rotation=<rot 90 0 0>}]
 dyn_ft_r_Vest_03_jnt [{shapeType=Capsule, length=13, radius=1.5, shapeName=dyn_ft_r_Vest_03_jnt_capsule, center=<7 2.5 0>, rotation=<rot 85 90 145>}]
 dyn_ft_r_Vest_MSTR [{shapeType=Capsule, length=0.505, radius=0.505, shapeName=dyn_ft_r_Vest_MSTR_capsule, rotation=<rot 0 -90 0>}]
 pelvis [{shapeType=Capsule, length=8, radius=10.75, shapeName=pelvis_capsule, center=<0 -1.021 0>}]
 spine_01 [{shapeType=Capsule, length=6, radius=10.5, shapeName=spine_01_capsule, center=<0 0.1674 0>}]
 spine_02 [{shapeType=Capsule, length=6, radius=9, shapeName=spine_02_capsule, center=<0 0.9223 0>}]
 spine_03 [{shapeType=Capsule, length=5, radius=10, shapeName=spine_03_capsule, center=<0 1.3842 0>}]
 spine_04 [{shapeType=Capsule, length=3, radius=12, shapeName=spine_04_capsule, center=<0 1.252 0>}]
 thigh_l [{shapeType=Capsule, length=25, radius=9.25, shapeName=thigh_l_capsule, center=<-20 1 0>, rotation=<rot 90 0 0>}]
 thigh_r [{shapeType=Capsule, length=25, radius=9.25, shapeName=thigh_r_capsule, center=<20 -1 0>, rotation=<rot 90 0 0>}]
}
constraints [
 {bone1Name=dyn_UpperStrap_01_jnt, bone2Name=dyn_UpperStrap_MSTR, swing1LimitDegrees=25, swing2LimitDegrees=25, twistLimitDegrees=2}
 {bone1Name=dyn_UpperStrap_02_jnt, bone2Name=dyn_UpperStrap_01_jnt, swing1LimitDegrees=25, swing2LimitDegrees=25, twistLimitDegrees=2}
 {bone1Name=dyn_UpperStrap_03_jnt, bone2Name=dyn_UpperStrap_02_jnt, swing1LimitDegrees=25, swing2LimitDegrees=25, twistLimitDegrees=2}
 {bone1Name=dyn_bk_Vest_01_jnt, bone2Name=dyn_bk_Vest_MSTR, swing1LimitDegrees=35, swing2LimitDegrees=2, twistLimitDegrees=2}
 {bone1Name=dyn_bk_Vest_02_jnt, bone2Name=dyn_bk_Vest_01_jnt, swing1LimitDegrees=35, swing2LimitDegrees=2, twistLimitDegrees=2}
 {bone1Name=dyn_bk_l_Vest_01_jnt, bone2Name=dyn_bk_l_Vest_MSTR, swing1LimitDegrees=2, swing2LimitDegrees=35, twistLimitDegrees=2}
 {bone1Name=dyn_bk_l_Vest_02_jnt, bone2Name=dyn_bk_l_Vest_01_jnt, swing1LimitDegrees=2, swing2LimitDegrees=35, twistLimitDegrees=2}
 {bone1Name=dyn_bk_r_Vest_01_jnt, bone2Name=dyn_bk_r_Vest_MSTR, swing1LimitDegrees=2, swing2LimitDegrees=35, twistLimitDegrees=2}
 {bone1Name=dyn_bk_r_Vest_02_jnt, bone2Name=dyn_bk_r_Vest_01_jnt, swing1LimitDegrees=2, swing2LimitDegrees=35, twistLimitDegrees=2}
 {bone1Name=dyn_ft_l_Vest_01_jnt, bone2Name=dyn_ft_l_Vest_MSTR, swing1LimitDegrees=35, swing2LimitDegrees=35, twistLimitDegrees=2}
 {bone1Name=dyn_ft_l_Vest_02_jnt, bone2Name=dyn_ft_l_Vest_01_jnt, swing1LimitDegrees=35, swing2LimitDegrees=35, twistLimitDegrees=2}
 {bone1Name=dyn_LowerStrap_MSTR, bone2Name=dyn_ft_l_Vest_02_jnt, swing1LimitDegrees=45, swing2LimitDegrees=45, twistLimitDegrees=2}
 {bone1Name=dyn_LowerStrap_01_jnt, bone2Name=dyn_LowerStrap_MSTR, swing1LimitDegrees=35, swing2LimitDegrees=35, twistLimitDegrees=5}
 {bone1Name=dyn_LowerStrap_02_jnt, bone2Name=dyn_LowerStrap_01_jnt, swing1LimitDegrees=35, swing2LimitDegrees=35, twistLimitDegrees=5}
 {bone1Name=dyn_LowerStrap_03_jnt, bone2Name=dyn_LowerStrap_02_jnt, swing1LimitDegrees=35, swing2LimitDegrees=35, twistLimitDegrees=5}
 {bone1Name=dyn_ft_l_Vest_03_jnt, bone2Name=dyn_ft_l_Vest_02_jnt, swing1LimitDegrees=35, swing2LimitDegrees=35, twistLimitDegrees=2}
 {bone1Name=dyn_ft_r_Vest_01_jnt, bone2Name=dyn_ft_r_Vest_MSTR, swing1LimitDegrees=35, swing2LimitDegrees=35, twistLimitDegrees=2}
 {bone1Name=dyn_ft_r_Vest_02_jnt, bone2Name=dyn_ft_r_Vest_01_jnt, swing1LimitDegrees=35, swing2LimitDegrees=35, twistLimitDegrees=2}
 {bone1Name=dyn_ft_r_Vest_03_jnt, bone2Name=dyn_ft_r_Vest_02_jnt, swing1LimitDegrees=35, swing2LimitDegrees=35, twistLimitDegrees=2}
 {
  bone1Name=spine_02
  bone2Name=spine_01
  swing1LimitDegrees=45
  swing1Motion=Free
  swing2LimitDegrees=45
  swing2Motion=Free
  twistLimitDegrees=45
  twistMotion=Free
 }
 {
  bone1Name=spine_03
  bone2Name=spine_01
  swing1LimitDegrees=45
  swing1Motion=Free
  swing2LimitDegrees=45
  swing2Motion=Free
  twistLimitDegrees=45
  twistMotion=Free
 }
 {
  bone1Name=spine_04
  bone2Name=spine_01
  swing1LimitDegrees=45
  swing1Motion=Free
  swing2LimitDegrees=45
  swing2Motion=Free
  twistLimitDegrees=45
  twistMotion=Free
 }
 {
  bone1Name=dyn_bk_r_Vest_MSTR
  bone2Name=spine_01
  swing1LimitDegrees=45
  swing1Motion=Free
  swing2LimitDegrees=45
  swing2Motion=Free
  twistLimitDegrees=45
  twistMotion=Free
 }
 {
  bone1Name=dyn_ft_r_Vest_MSTR
  bone2Name=spine_01
  swing1LimitDegrees=45
  swing1Motion=Free
  swing2LimitDegrees=45
  swing2Motion=Free
  twistLimitDegrees=45
  twistMotion=Free
 }
 {
  bone1Name=dyn_bk_Vest_MSTR
  bone2Name=spine_01
  swing1LimitDegrees=45
  swing1Motion=Free
  swing2LimitDegrees=45
  swing2Motion=Free
  twistLimitDegrees=45
  twistMotion=Free
 }
 {
  bone1Name=dyn_bk_l_Vest_MSTR
  bone2Name=spine_01
  swing1LimitDegrees=45
  swing1Motion=Free
  swing2LimitDegrees=45
  swing2Motion=Free
  twistLimitDegrees=45
  twistMotion=Free
 }
 {
  bone1Name=dyn_ft_l_Vest_MSTR
  bone2Name=spine_01
  swing1LimitDegrees=45
  swing1Motion=Free
  swing2LimitDegrees=45
  swing2Motion=Free
  twistLimitDegrees=45
  twistMotion=Free
 }
 {
  bone1Name=dyn_UpperStrap_MSTR
  bone2Name=spine_01
  swing1LimitDegrees=45
  swing1Motion=Free
  swing2LimitDegrees=45
  swing2Motion=Free
  twistLimitDegrees=45
  twistMotion=Free
 }
 {
  bone1Name=spine_01
  bone2Name=pelvis
  swing1LimitDegrees=45
  swing1Motion=Free
  swing2LimitDegrees=45
  swing2Motion=Free
  twistLimitDegrees=45
  twistMotion=Free
 }
 {
  bone1Name=thigh_l
  bone2Name=pelvis
  swing1LimitDegrees=45
  swing1Motion=Free
  swing2LimitDegrees=45
  swing2Motion=Free
  twistLimitDegrees=45
  twistMotion=Free
 }
 {
  bone1Name=thigh_r
  bone2Name=pelvis
  swing1LimitDegrees=45
  swing1Motion=Free
  swing2LimitDegrees=45
  swing2Motion=Free
  twistLimitDegrees=45
  twistMotion=Free
 }
]
