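# Post-process style AnimBP for the HydroIgnite body cosmetic, adding secondary motion to its vest, collar, strap and shoelaces.
# Copied parent pose passes through bone constraints, Control Rig and AnimDynamics into a cache.
# RigidBody simulation uses BackPin or Vest physics depending on backbling, followed by additive bone offsets for movement states.
# EventGraph converts pawn states into alphas, damping vest simulation, and drives gravity override and material WPO/skydive wind.
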AnimBlueprint /BRCosmetics/Characters/Player/Male/Medium/Bodies/M_MED_HydroIgnite/Meshes/M_MED_HydroIgnite_AnimBP
target_skeleton=&/M_MED_HydroIgnite_Skeleton
graphs [
 AnimGraph {nodes [
   n1 = RigidBody!(
    ComponentPose=Local To Component(LocalPose=Use cached pose 'PreSim Cache'())
    Alpha=Get VestAlpha() -> (VestAlpha:double)
    OverrideWorldGravity:Vector=Get GravityOverride()
    ExternalForce
   ) -> (Pose) {
    bClampLinearTranslationLimitToRefPose=true
    bOverrideWorldGravity=true
    overridePhysicsAsset=&/M_MED_HydroIgnite_Vest_Physics
    simulationSpace=BaseBoneSpace
    alphaBoolBlend {blendOption=Linear}
    alphaScaleBiasClamp {interpSpeedIncreasing=2}
    baseBoneRef {boneName=pelvis}
    componentAppliedLinearAccClamp=<3000 3000 3000>
    componentLinearAccScale=<0.4 0.3 0.3>
    componentLinearVelScale=<0.4 0.3 0.3>
    simSpaceSettings {
     maxAngularAcceleration=1000
     maxAngularVelocity=3
     maxLinearAcceleration=700
     maxLinearVelocity=500
     velocityScaleZ=0
     worldAlpha=0.25
     externalAngularVelocity {z=0.1}
     externalLinearDragV=<0.2 0.2>
    }
   }
   n2 = Component To Local(ComponentPose=@n1)
   n3 = RigidBody(
    ComponentPose=Local To Component(LocalPose=Use cached pose 'PreSim Cache'())
    Alpha=Get Backbling Physics Alpha() -> (Backbling Physics Alpha:float)
    OverrideWorldGravity=Get GravityOverride()
   ) {
    bClampLinearTranslationLimitToRefPose=true
    bOverrideWorldGravity=true
    overridePhysicsAsset=&/M_MED_HydroIgnite_BackPin_Physics
    simulationSpace=BaseBoneSpace
    alphaBoolBlend {blendOption=Linear}
    alphaScaleBiasClamp {interpSpeedIncreasing=2}
    baseBoneRef {boneName=pelvis}
    componentAppliedLinearAccClamp=<3000 3000 3000>
    componentLinearAccScale=<0.4 0.3 0.3>
    componentLinearVelScale=<0.4 0.3 0.3>
    simSpaceSettings {
     maxAngularAcceleration=1000
     maxAngularVelocity=3
     maxLinearAcceleration=700
     maxLinearVelocity=500
     velocityScaleZ=0
     worldAlpha=0.25
     externalAngularVelocity {z=0.1}
     externalLinearDragV=<0.2 0.2>
    }
   }
   n4 = Get Is Jumping Float() -> (Is Jumping Float:double)
   n5 = "Transform (Modify) Bone - Bone: dyn_bk_Vest_02_jnt"(ComponentPose=@n3, Alpha=@n4, Translation="0,0,0", Rotation="0,25,0", Scale="1,1,1") {
    rotationMode=BMM_Additive
    rotationSpace=BCS_BoneSpace
    translationMode=BMM_Additive
    translationSpace=BCS_BoneSpace
    alphaBoolBlend {blendOption=Linear}
    alphaScaleBiasClamp {bInterpResult=true, interpSpeedDecreasing=2}
    boneToModify {boneName=dyn_bk_Vest_02_jnt}
    rotation {yaw=25}
   }
   n6 = "Transform (Modify) Bone - Bone: dyn_bk_r_Vest_02_jnt"(ComponentPose=@n5, Alpha=@n4, Translation="0,0,0", Rotation="-25,-25,0", Scale="1,1,1") {
    rotationMode=BMM_Additive
    rotationSpace=BCS_BoneSpace
    translationMode=BMM_Additive
    translationSpace=BCS_BoneSpace
    alphaBoolBlend {blendOption=Linear}
    alphaScaleBiasClamp {bInterpResult=true, interpSpeedDecreasing=2}
    boneToModify {boneName=dyn_bk_r_Vest_02_jnt}
    rotation {pitch=-25, yaw=-25}
   }
   n7 = "Transform (Modify) Bone - Bone: dyn_bk_l_Vest_02_jnt"(ComponentPose=@n6, Alpha=@n4, Translation="0,0,0", Rotation="-25,25,0", Scale="1,1,1") {
    rotationMode=BMM_Additive
    rotationSpace=BCS_BoneSpace
    translationMode=BMM_Additive
    translationSpace=BCS_BoneSpace
    alphaBoolBlend {blendOption=Linear}
    alphaScaleBiasClamp {bInterpResult=true, interpSpeedDecreasing=2}
    boneToModify {boneName=dyn_bk_l_Vest_02_jnt}
    rotation {pitch=-25, yaw=25}
   }
   n8 = Component To Local(ComponentPose=@n7)
   n9 = Get Is Wearing Backbling()
   n10 = Get Is Jumping Float()
   n11 = "Transform (Modify) Bone - Bone: dyn_bk_r_Vest_MSTR"(
    ComponentPose=Local To Component(LocalPose=Blend Poses by bool(BlendPose_0=@n8, BlendPose_1=@n2, bActiveValue=@n9) {blendType=Linear})
    Alpha=@n10
    Translation="0,5,0"
    Rotation="0,0,0"
    Scale="1,1,1"
   ) {
    rotationMode=BMM_Additive
    rotationSpace=BCS_BoneSpace
    translationMode=BMM_Additive
    translationSpace=BCS_BoneSpace
    alphaBoolBlend {blendOption=Linear}
    alphaScaleBiasClamp {bInterpResult=true, interpSpeedDecreasing=2}
    boneToModify {boneName=dyn_bk_r_Vest_MSTR}
    translation {y=5}
   }
   n12 = "Transform (Modify) Bone - Bone: dyn_bk_l_Vest_MSTR"(ComponentPose=@n11, Alpha=@n10, Translation="0,-5,0", Rotation="0,0,0", Scale="1,1,1") {
    rotationMode=BMM_Additive
    rotationSpace=BCS_BoneSpace
    translationMode=BMM_Additive
    translationSpace=BCS_BoneSpace
    alphaBoolBlend {blendOption=Linear}
    alphaScaleBiasClamp {bInterpResult=true, interpSpeedDecreasing=2}
    boneToModify {boneName=dyn_bk_l_Vest_MSTR}
    translation {y=-5}
   }
   n13 = Get Is Tactical Sprinting Float() -> (Is Tactical Sprinting Float:double)
   n14 = "Transform (Modify) Bone - Bone: dyn_bk_r_Vest_MSTR"(ComponentPose=@n12, Alpha=@n13, Translation="0,1,0", Rotation="0,-15,0", Scale="1,1,1") {
    rotationMode=BMM_Additive
    rotationSpace=BCS_BoneSpace
    translationMode=BMM_Additive
    translationSpace=BCS_BoneSpace
    alphaBoolBlend {blendOption=Linear}
    alphaScaleBiasClamp {bInterpResult=true, interpSpeedDecreasing=2}
    boneToModify {boneName=dyn_bk_r_Vest_MSTR}
    rotation {yaw=-15}
    translation {y=1}
   }
   n15 = "Transform (Modify) Bone - Bone: dyn_bk_l_Vest_MSTR"(ComponentPose=@n14, Alpha=@n13, Translation="0,-1,0", Rotation="0,15,0", Scale="1,1,1") {
    rotationMode=BMM_Additive
    rotationSpace=BCS_BoneSpace
    translationMode=BMM_Additive
    translationSpace=BCS_BoneSpace
    alphaBoolBlend {blendOption=Linear}
    alphaScaleBiasClamp {bInterpResult=true, interpSpeedDecreasing=2}
    boneToModify {boneName=dyn_bk_l_Vest_MSTR}
    rotation {yaw=15}
    translation {y=-1}
   }
   n16 = Get Is swimming Float() -> (Is swimming Float:double)
   n17 = "Transform (Modify) Bone - Bone: dyn_bk_r_Vest_MSTR"(ComponentPose=@n15, Alpha=@n16, Translation="0,3,0", Rotation="0,0,0", Scale="1,1,1") {
    rotationMode=BMM_Additive
    rotationSpace=BCS_BoneSpace
    translationMode=BMM_Additive
    translationSpace=BCS_BoneSpace
    alphaBoolBlend {blendOption=Linear}
    alphaScaleBiasClamp {bInterpResult=true, interpSpeedDecreasing=2}
    boneToModify {boneName=dyn_bk_r_Vest_MSTR}
    translation {y=3}
   }
   n18 = "Transform (Modify) Bone - Bone: dyn_bk_l_Vest_MSTR"(ComponentPose=@n17, Alpha=@n16, Translation="0,-3,0", Rotation="0,0,0", Scale="1,1,1") {
    rotationMode=BMM_Additive
    rotationSpace=BCS_BoneSpace
    translationMode=BMM_Additive
    translationSpace=BCS_BoneSpace
    alphaBoolBlend {blendOption=Linear}
    alphaScaleBiasClamp {bInterpResult=true, interpSpeedDecreasing=2}
    boneToModify {boneName=dyn_bk_l_Vest_MSTR}
    translation {y=-3}
   }
   n19 = Get Is on Grind Rail Float() -> (Is on Grind Rail Float:double)
   n20 = "Transform (Modify) Bone - Bone: dyn_bk_r_Vest_MSTR"(ComponentPose=@n18, Alpha=@n19, Translation="0,3,2", Rotation="0,0,0", Scale="1,1,1") {
    rotationMode=BMM_Additive
    rotationSpace=BCS_BoneSpace
    translationMode=BMM_Additive
    translationSpace=BCS_BoneSpace
    alphaBoolBlend {blendOption=Linear}
    alphaScaleBiasClamp {bInterpResult=true, interpSpeedDecreasing=2}
    boneToModify {boneName=dyn_bk_r_Vest_MSTR}
    translation {y=3, z=2}
   }
   n21 = "Transform (Modify) Bone - Bone: dyn_bk_l_Vest_MSTR"(ComponentPose=@n20, Alpha=@n19, Translation="0,-3,2", Rotation="0,0,0", Scale="1,1,1") {
    rotationMode=BMM_Additive
    rotationSpace=BCS_BoneSpace
    translationMode=BMM_Additive
    translationSpace=BCS_BoneSpace
    alphaBoolBlend {blendOption=Linear}
    alphaScaleBiasClamp {bInterpResult=true, interpSpeedDecreasing=2}
    boneToModify {boneName=dyn_bk_l_Vest_MSTR}
    translation {y=-3, z=2}
   }
   n22 = "Transform (Modify) Bone - Bone: dyn_bk_Vest_MSTR"(ComponentPose=@n21, Alpha=@n19, Translation="0,0,3", Rotation="0,0,0", Scale="1,1,1") {
    rotationMode=BMM_Additive
    rotationSpace=BCS_BoneSpace
    translationMode=BMM_Additive
    translationSpace=BCS_BoneSpace
    alphaBoolBlend {blendOption=Linear}
    alphaScaleBiasClamp {bInterpResult=true, interpSpeedDecreasing=2}
    boneToModify {boneName=dyn_bk_Vest_MSTR}
    translation {z=3}
   }
   n23 = "Transform (Modify) Bone - Bone: spine_05_back"(ComponentPose=@n22, Alpha=@n19, Translation="0,-2.5,0", Rotation="0,0,0", Scale="1,1,1") {
    rotationMode=BMM_Additive
    rotationSpace=BCS_BoneSpace
    translationMode=BMM_Additive
    translationSpace=BCS_BoneSpace
    alphaBoolBlend {blendOption=Linear}
    boneToModify {boneName=spine_05_back}
    translation {y=-2.5}
   }
   n24 = Get Is Skydive Floating Alpha() -> (Is Skydive Floating Alpha:double)
   n25 = "Transform (Modify) Bone - Bone: dyn_bk_r_Vest_MSTR"(ComponentPose=@n23, Alpha=@n24, Translation="0,1,1", Rotation="10,10,0", Scale="1,1,1") {
    rotationMode=BMM_Additive
    rotationSpace=BCS_BoneSpace
    translationMode=BMM_Additive
    translationSpace=BCS_BoneSpace
    alphaBoolBlend {blendOption=Linear}
    alphaScaleBiasClamp {bInterpResult=true, interpSpeedDecreasing=2}
    boneToModify {boneName=dyn_bk_r_Vest_MSTR}
    rotation {pitch=10, yaw=10}
    translation {y=1, z=1}
   }
   n26 = "Transform (Modify) Bone - Bone: dyn_bk_l_Vest_MSTR"(ComponentPose=@n25, Alpha=@n24, Translation="0,1,1", Rotation="10,-10,0", Scale="1,1,1") {
    rotationMode=BMM_Additive
    rotationSpace=BCS_BoneSpace
    translationMode=BMM_Additive
    translationSpace=BCS_BoneSpace
    alphaBoolBlend {blendOption=Linear}
    alphaScaleBiasClamp {bInterpResult=true, interpSpeedDecreasing=2}
    boneToModify {boneName=dyn_bk_l_Vest_MSTR}
    rotation {pitch=10, yaw=-10}
    translation {y=1, z=1}
   }
   n27 = "Transform (Modify) Bone - Bone: dyn_bk_Vest_MSTR"(ComponentPose=@n26, Alpha=@n24, Translation="0,1,1", Rotation="0,-20,0", Scale="1,1,1") {
    rotationMode=BMM_Additive
    rotationSpace=BCS_BoneSpace
    translationMode=BMM_Additive
    translationSpace=BCS_BoneSpace
    alphaBoolBlend {blendOption=Linear}
    alphaScaleBiasClamp {bInterpResult=true, interpSpeedDecreasing=2}
    boneToModify {boneName=dyn_bk_Vest_MSTR}
    rotation {yaw=-20}
    translation {y=1, z=1}
   }
   n28 = "Transform (Modify) Bone - Bone: dyn_Collar_c_MSTR"(ComponentPose=@n27, Alpha=@n24, Translation="3,0,-3", Rotation="50,0,0", Scale="1,1,1") {
    rotationMode=BMM_Additive
    rotationSpace=BCS_BoneSpace
    translationMode=BMM_Additive
    translationSpace=BCS_BoneSpace
    alphaBoolBlend {blendOption=Linear}
    alphaScaleBiasClamp {bInterpResult=true, interpSpeedDecreasing=2}
    boneToModify {boneName=dyn_Collar_c_MSTR}
    rotation {pitch=50}
    translation {x=3, z=-3}
   }
   n29 = "Transform (Modify) Bone - Bone: spine_05_back"(ComponentPose=@n28, Alpha=@n24, Translation="0,-2.5,0", Rotation="0,0,0", Scale="1,1,1") {
    rotationMode=BMM_Additive
    rotationSpace=BCS_BoneSpace
    translationMode=BMM_Additive
    translationSpace=BCS_BoneSpace
    alphaBoolBlend {blendOption=Linear}
    boneToModify {boneName=spine_05_back}
    translation {y=-2.5}
   }
   n30 = Get Is Playing Melee Anim Float() -> (Is Playing Melee Anim Float:double)
   n31 = "Transform (Modify) Bone - Bone: dyn_bk_r_Vest_MSTR"(ComponentPose=@n29, Alpha=@n30, Translation="0,1,1", Rotation="10,10,0", Scale="1,1,1") {
    rotationMode=BMM_Additive
    rotationSpace=BCS_BoneSpace
    translationMode=BMM_Additive
    translationSpace=BCS_BoneSpace
    alphaBoolBlend {blendOption=Linear}
    alphaScaleBiasClamp {bInterpResult=true, interpSpeedDecreasing=2}
    boneToModify {boneName=dyn_bk_r_Vest_MSTR}
    rotation {pitch=10, yaw=10}
    translation {y=1, z=1}
   }
   n32 = "Transform (Modify) Bone - Bone: dyn_bk_l_Vest_MSTR"(ComponentPose=@n31, Alpha=@n30, Translation="0,1,1", Rotation="10,-10,0", Scale="1,1,1") {
    rotationMode=BMM_Additive
    rotationSpace=BCS_BoneSpace
    translationMode=BMM_Additive
    translationSpace=BCS_BoneSpace
    alphaBoolBlend {blendOption=Linear}
    alphaScaleBiasClamp {bInterpResult=true, interpSpeedDecreasing=2}
    boneToModify {boneName=dyn_bk_l_Vest_MSTR}
    rotation {pitch=10, yaw=-10}
    translation {y=1, z=1}
   }
   n33 = "Transform (Modify) Bone - Bone: spine_05_back"(ComponentPose=@n32, Alpha=1, Translation="0,-2.5,0", Rotation="0,0,0", Scale="1,1,1") {
    rotationMode=BMM_Additive
    rotationSpace=BCS_BoneSpace
    translationMode=BMM_Additive
    translationSpace=BCS_BoneSpace
    alphaBoolBlend {blendOption=Linear}
    boneToModify {boneName=spine_05_back}
    translation {y=-2.5}
   }
   Output Pose(Result=Component To Local(ComponentPose=@n33)) {layerGroup=None}
   n34 = "Constraint - dyn_bk_r_Vest_MSTR (PPP)"!(
    ComponentPose=Local To Component(LocalPose=Copy Pose From Mesh() {bUseAttachedParent=true})
    Alpha=1
    ConstraintWeights_0=0.7
    ConstraintWeights_1:float=0.1
    ConstraintWeights_2:float=0.2
   ) -> (Pose) {
    alphaBoolBlend {blendOption=Linear}
    boneToModify {boneName=dyn_bk_r_Vest_MSTR}
    constraintSetup [{targetBone {boneName=spine_04}}, {targetBone {boneName=spine_05}}, {targetBone {boneName=clavicle_r}}]
    constraintWeights [1,1,1]
   }
   n35 = "Constraint - dyn_ft_r_Vest_MSTR (PP)"!(ComponentPose=@n34, Alpha=1, ConstraintWeights_0=0.9, ConstraintWeights_1:float=0.1) -> (Pose) {
    alphaBoolBlend {blendOption=Linear}
    boneToModify {boneName=dyn_ft_r_Vest_MSTR}
    constraintSetup [{targetBone {boneName=spine_04}}, {targetBone {boneName=spine_05}}]
    constraintWeights [1,1]
   }
   n36 = "Constraint - dyn_bk_Vest_MSTR (PP)"!(ComponentPose=@n35, Alpha=1, ConstraintWeights_0=0.9, ConstraintWeights_1:float=0.1) -> (Pose) {
    alphaBoolBlend {blendOption=Linear}
    boneToModify {boneName=dyn_bk_Vest_MSTR}
    constraintSetup [{targetBone {boneName=spine_04}}, {targetBone {boneName=spine_05}}]
    constraintWeights [1,1]
   }
   n37 = "Constraint - dyn_bk_l_Vest_MSTR (PPP)"!(ComponentPose=@n36, Alpha=1, ConstraintWeights_0=0.7, ConstraintWeights_1:float=0.1, ConstraintWeights_2:float=0.2) -> (Pose) {
    alphaBoolBlend {blendOption=Linear}
    boneToModify {boneName=dyn_bk_l_Vest_MSTR}
    constraintSetup [{targetBone {boneName=spine_04}}, {targetBone {boneName=spine_05}}, {targetBone {boneName=clavicle_l}}]
    constraintWeights [1,1,1]
   }
   n38 = "Constraint - dyn_ft_l_Vest_MSTR (PP)"!(ComponentPose=@n37, Alpha=1, ConstraintWeights_0=0.9, ConstraintWeights_1:float=0.1) -> (Pose) {
    alphaBoolBlend {blendOption=Linear}
    boneToModify {boneName=dyn_ft_l_Vest_MSTR}
    constraintSetup [{targetBone {boneName=spine_04}}, {targetBone {boneName=spine_05}}]
    constraintWeights [1,1]
   }
   n39 = "Constraint - spine_05_back (PPP)"!(ComponentPose=@n38, Alpha=1, ConstraintWeights_0=0.4, ConstraintWeights_1:float=0.2, ConstraintWeights_2:float=0.4) -> (Pose) {
    alphaBoolBlend {blendOption=Linear}
    boneToModify {boneName=spine_05_back}
    constraintSetup [{targetBone {boneName=clavicle_l}}, {targetBone {boneName=spine_05}}, {targetBone {boneName=clavicle_r}}]
    constraintWeights [1,1,1]
   }
   n40 = "Constraint - dyn_Collar_r_MSTR (RR)"!(ComponentPose=@n39, Alpha=1, ConstraintWeights_0=0.7, ConstraintWeights_1:float=0.3) -> (Pose) {
    alphaBoolBlend {blendOption=Linear}
    boneToModify {boneName=dyn_Collar_r_MSTR}
    constraintSetup [{transformType=Rotation, targetBone {boneName=neck_01}}, {transformType=Rotation, targetBone {boneName=spine_05}}]
    constraintWeights [1,1]
   }
   n41 = "Constraint - dyn_Collar_c_MSTR (RR)"!(ComponentPose=@n40, Alpha=1, ConstraintWeights_0=0.7, ConstraintWeights_1:float=0.3) -> (Pose) {
    alphaBoolBlend {blendOption=Linear}
    boneToModify {boneName=dyn_Collar_c_MSTR}
    constraintSetup [{transformType=Rotation, targetBone {boneName=neck_01}}, {transformType=Rotation, targetBone {boneName=spine_05}}]
    constraintWeights [1,1]
   }
   n42 = "Constraint - dyn_Collar_l_MSTR (RR)"!(ComponentPose=@n41, Alpha=1, ConstraintWeights_0=0.7, ConstraintWeights_1:float=0.3) -> (Pose) {
    alphaBoolBlend {blendOption=Linear}
    boneToModify {boneName=dyn_Collar_l_MSTR}
    constraintSetup [{transformType=Rotation, targetBone {boneName=neck_01}}, {transformType=Rotation, targetBone {boneName=spine_05}}]
    constraintWeights [1,1]
   }
   n43 = Control Rig!(
    Source=Component To Local(ComponentPose=@n42)
    Alpha
    Is Pistol Alpha:double=Get Is Pistol Alpha() -> (Is Pistol Alpha:double)
    Is Rifle Alpha:double=Get Is Rifle Alpha() -> (Is Rifle Alpha:double)
    Is Crouching Alpha:double=Get Is Crouching Float() -> (Is Crouching Float:double)
   ) -> (Pose) {
    alphaBoolBlend {blendOption=Linear}
    controlRigAssetReference {blueprintRigClass=&/M_MED_HydroIgnite_CtrlRig.M_MED_HydroIgnite_CtrlRig_C}
   }
   n44 = "Transform (Modify) Bone - Bone: dyn_bk_r_Vest_MSTR"(ComponentPose=Local To Component(LocalPose=@n43), Alpha=1, Translation="0,0,1", Rotation="0,0,0", Scale="1,1,1") {
    rotationMode=BMM_Additive
    rotationSpace=BCS_BoneSpace
    translationMode=BMM_Additive
    translationSpace=BCS_BoneSpace
    alphaBoolBlend {blendOption=Linear}
    alphaScaleBiasClamp {bInterpResult=true, interpSpeedDecreasing=2}
    boneToModify {boneName=dyn_bk_r_Vest_MSTR}
    translation {z=1}
   }
   n45 = "Transform (Modify) Bone - Bone: dyn_bk_l_Vest_MSTR"(ComponentPose=@n44, Alpha=1, Translation="0,0,1", Rotation="0,0,0", Scale="1,1,1") {
    rotationMode=BMM_Additive
    rotationSpace=BCS_BoneSpace
    translationMode=BMM_Additive
    translationSpace=BCS_BoneSpace
    alphaBoolBlend {blendOption=Linear}
    alphaScaleBiasClamp {bInterpResult=true, interpSpeedDecreasing=2}
    boneToModify {boneName=dyn_bk_l_Vest_MSTR}
    translation {z=1}
   }
   n46 = "Transform (Modify) Bone - Bone: dyn_bk_Vest_MSTR"(ComponentPose=@n45, Alpha=1, Translation="0,0,1", Rotation="0,0,0", Scale="1,1,1") {
    rotationMode=BMM_Additive
    rotationSpace=BCS_BoneSpace
    translationMode=BMM_Additive
    translationSpace=BCS_BoneSpace
    alphaBoolBlend {blendOption=Linear}
    alphaScaleBiasClamp {bInterpResult=true, interpSpeedDecreasing=2}
    boneToModify {boneName=dyn_bk_Vest_MSTR}
    translation {z=1}
   }
   n47 = Anim Dynamics - Bone: dyn_l_Shoelace_l_Strand_01_jnt!(ComponentPose=@n46, Alpha=1, GravityOverride:Vector="0,0,-980", ExternalForce="0,0,0") -> (Pose) {
    angularDampingOverride=1
    angularSpringConstant=800
    bAngularSpring=true
    bOverrideAngularDamping=true
    bUseGravityOverride=true
    simulationSpace=RootRelative
    alphaBoolBlend {blendOption=Linear}
    boundBone {boneName=dyn_l_Shoelace_l_Strand_01_jnt}
    componentAppliedLinearAccClamp=<3000 3000 3000>
    componentLinearAccScale=<0.3 0.3 0.3>
    componentLinearVelScale=<0.3 0.3 0.3>
    gravityOverride {z=-980}
    physicsBodyDefinitions [{
      boundBone {boneName=dyn_l_Shoelace_l_Strand_01_jnt}
      boxExtents=<10 10 10>
      constraintSetup {angularLimitsMax=<5 10 15>, angularLimitsMin=<-5 -15 -15>, angularTarget=<1 0 0>}
      localJointOffset=<4 0 0>
     }]
    relativeSpaceBone {boneName=calf_l}
   }
   n48 = Anim Dynamics - Bone: dyn_l_Shoelace_r_Strand_01_jnt!(ComponentPose=@n47, Alpha=1, GravityOverride:Vector="0,0,-980", ExternalForce="0,0,0") -> (Pose) {
    angularDampingOverride=1
    angularSpringConstant=100
    bAngularSpring=true
    bOverrideAngularDamping=true
    bUseGravityOverride=true
    simulationSpace=RootRelative
    alphaBoolBlend {blendOption=Linear}
    boundBone {boneName=dyn_l_Shoelace_r_Strand_01_jnt}
    componentAppliedLinearAccClamp=<3000 3000 3000>
    componentLinearAccScale=<0.3 0.3 0.3>
    componentLinearVelScale=<0.3 0.3 0.3>
    gravityOverride {z=-980}
    physicsBodyDefinitions [{
      boundBone {boneName=dyn_l_Shoelace_r_Strand_01_jnt}
      boxExtents=<10 10 10>
      constraintSetup {angularLimitsMax=<5 10 15>, angularLimitsMin=<-5 -15 -15>, angularTarget=<1 0 0>}
      localJointOffset=<4 0 0>
     }]
    relativeSpaceBone {boneName=calf_l}
   }
   n49 = Anim Dynamics - Bone: dyn_l_Shoelace_l_Strand_02_jnt!(ComponentPose=@n48, Alpha=1, GravityOverride:Vector="0,0,-980", ExternalForce="0,0,0") -> (Pose) {
    angularDampingOverride=1
    angularSpringConstant=800
    bAngularSpring=true
    bOverrideAngularDamping=true
    bUseGravityOverride=true
    simulationSpace=RootRelative
    alphaBoolBlend {blendOption=Linear}
    boundBone {boneName=dyn_l_Shoelace_l_Strand_02_jnt}
    componentAppliedLinearAccClamp=<3000 3000 3000>
    componentLinearAccScale=<0.3 0.3 0.3>
    componentLinearVelScale=<0.3 0.3 0.3>
    gravityOverride {z=-980}
    physicsBodyDefinitions [{
      boundBone {boneName=dyn_l_Shoelace_l_Strand_02_jnt}
      boxExtents=<10 10 10>
      constraintSetup {angularLimitsMax=<5 10 15>, angularLimitsMin=<-5 -15 -15>, angularTarget=<1 0 0>}
      localJointOffset=<4 0 0>
     }]
    relativeSpaceBone {boneName=calf_l}
   }
   n50 = Anim Dynamics - Bone: dyn_l_Shoelace_r_Strand_02_jnt!(ComponentPose=@n49, Alpha=1, GravityOverride:Vector="0,0,-980", ExternalForce="0,0,0") -> (Pose) {
    angularDampingOverride=1
    angularSpringConstant=100
    bAngularSpring=true
    bOverrideAngularDamping=true
    bUseGravityOverride=true
    simulationSpace=RootRelative
    alphaBoolBlend {blendOption=Linear}
    boundBone {boneName=dyn_l_Shoelace_r_Strand_02_jnt}
    componentAppliedLinearAccClamp=<3000 3000 3000>
    componentLinearAccScale=<0.3 0.3 0.3>
    componentLinearVelScale=<0.3 0.3 0.3>
    gravityOverride {z=-980}
    physicsBodyDefinitions [{
      boundBone {boneName=dyn_l_Shoelace_r_Strand_02_jnt}
      boxExtents=<10 10 10>
      constraintSetup {angularLimitsMax=<5 10 15>, angularLimitsMin=<-5 -15 -15>, angularTarget=<1 0 0>}
      localJointOffset=<4 0 0>
     }]
    relativeSpaceBone {boneName=calf_l}
   }
   n51 = Anim Dynamics - Bone: dyn_r_Shoelace_r_Strand_01_jnt!(ComponentPose=@n50, Alpha=1, GravityOverride:Vector="0,0,-980", ExternalForce="0,0,0") -> (Pose) {
    angularDampingOverride=1
    angularSpringConstant=100
    bAngularSpring=true
    bOverrideAngularDamping=true
    bUseGravityOverride=true
    simulationSpace=RootRelative
    alphaBoolBlend {blendOption=Linear}
    boundBone {boneName=dyn_r_Shoelace_r_Strand_01_jnt}
    componentAppliedLinearAccClamp=<3000 3000 3000>
    componentLinearAccScale=<0.3 0.3 0.3>
    componentLinearVelScale=<0.3 0.3 0.3>
    gravityOverride {z=-980}
    physicsBodyDefinitions [{
      boundBone {boneName=dyn_r_Shoelace_r_Strand_01_jnt}
      boxExtents=<10 10 10>
      constraintSetup {angularLimitsMax=<5 10 15>, angularLimitsMin=<-5 -15 -15>, angularTarget=<1 0 0>}
      localJointOffset=<4 0 0>
     }]
    relativeSpaceBone {boneName=calf_r}
   }
   n52 = Anim Dynamics - Bone: dyn_r_Shoelace_l_Strand_01_jnt!(ComponentPose=@n51, Alpha=1, GravityOverride:Vector="0,0,-980", ExternalForce="0,0,0") -> (Pose) {
    angularDampingOverride=1
    angularSpringConstant=800
    bAngularSpring=true
    bOverrideAngularDamping=true
    bUseGravityOverride=true
    simulationSpace=RootRelative
    alphaBoolBlend {blendOption=Linear}
    boundBone {boneName=dyn_r_Shoelace_l_Strand_01_jnt}
    componentAppliedLinearAccClamp=<3000 3000 3000>
    componentLinearAccScale=<0.3 0.3 0.3>
    componentLinearVelScale=<0.3 0.3 0.3>
    gravityOverride {z=-980}
    physicsBodyDefinitions [{
      boundBone {boneName=dyn_r_Shoelace_l_Strand_01_jnt}
      boxExtents=<10 10 10>
      constraintSetup {angularLimitsMax=<5 10 15>, angularLimitsMin=<-5 -15 -15>, angularTarget=<1 0 0>}
      localJointOffset=<4 0 0>
     }]
    relativeSpaceBone {boneName=calf_r}
   }
   n53 = Anim Dynamics - Bone: dyn_r_Shoelace_r_Strand_02_jnt!(ComponentPose=@n52, Alpha=1, GravityOverride:Vector="0,0,-980", ExternalForce="0,0,0") -> (Pose) {
    angularDampingOverride=1
    angularSpringConstant=100
    bAngularSpring=true
    bOverrideAngularDamping=true
    bUseGravityOverride=true
    simulationSpace=RootRelative
    alphaBoolBlend {blendOption=Linear}
    boundBone {boneName=dyn_r_Shoelace_r_Strand_02_jnt}
    componentAppliedLinearAccClamp=<3000 3000 3000>
    componentLinearAccScale=<0.3 0.3 0.3>
    componentLinearVelScale=<0.3 0.3 0.3>
    gravityOverride {z=-980}
    physicsBodyDefinitions [{
      boundBone {boneName=dyn_r_Shoelace_r_Strand_02_jnt}
      boxExtents=<10 10 10>
      constraintSetup {angularLimitsMax=<5 10 15>, angularLimitsMin=<-5 -15 -15>, angularTarget=<1 0 0>}
      localJointOffset=<4 0 0>
     }]
    relativeSpaceBone {boneName=calf_r}
   }
   n54 = Anim Dynamics - Bone: dyn_r_Shoelace_l_Strand_02_jnt!(ComponentPose=@n53, Alpha=1, GravityOverride:Vector="0,0,-980", ExternalForce="0,0,0") -> (Pose) {
    angularDampingOverride=1
    angularSpringConstant=800
    bAngularSpring=true
    bOverrideAngularDamping=true
    bUseGravityOverride=true
    simulationSpace=RootRelative
    alphaBoolBlend {blendOption=Linear}
    boundBone {boneName=dyn_r_Shoelace_l_Strand_02_jnt}
    componentAppliedLinearAccClamp=<3000 3000 3000>
    componentLinearAccScale=<0.3 0.3 0.3>
    componentLinearVelScale=<0.3 0.3 0.3>
    gravityOverride {z=-980}
    physicsBodyDefinitions [{
      boundBone {boneName=dyn_r_Shoelace_l_Strand_02_jnt}
      boxExtents=<10 10 10>
      constraintSetup {angularLimitsMax=<5 10 15>, angularLimitsMin=<-5 -15 -15>, angularTarget=<1 0 0>}
      localJointOffset=<4 0 0>
     }]
    relativeSpaceBone {boneName=calf_r}
   }
   n55 = "Transform (Modify) Bone - Bone: dyn_LowerStrap_MSTR"(ComponentPose=@n54, Alpha=Get Is Crouching Float(), Translation="-1,-1,-1", Rotation="0,-15,15", Scale="1,1,1") {
    rotationMode=BMM_Additive
    rotationSpace=BCS_BoneSpace
    translationMode=BMM_Additive
    translationSpace=BCS_BoneSpace
    alphaBoolBlend {blendOption=Linear}
    alphaScaleBiasClamp {bInterpResult=true, interpSpeedDecreasing=2}
    boneToModify {boneName=dyn_LowerStrap_MSTR}
    rotation {roll=15, yaw=-15}
    translation=<-1 -1 -1>
   }
   Save cached pose 'PreSim Cache'(Pose=Component To Local(ComponentPose=@n55)) {cacheName=PreSim Cache}
  ]}
 EventGraph {nodes [
   n1 = Event BlueprintUpdateAnimation()
   > n2 = Sequence!(execute) -> (
    then_0
    then_1
    then_2:Exec
    then_3:Exec
    then_4:Exec
    then_5:Exec
    then_6:Exec
    then_7:Exec
    then_8:Exec
    then_9:Exec
    then_10:Exec
    then_11:Exec
    then_12:Exec
    then_13:Exec
   )
   > n3 = Gravity Override State Function DEPRECATED(CharacterPartAnimInstance=Self-Reference(), GameTimeInSeconds=GetGameTimeInSeconds(), GravityOverrideParams=Get Gravity Override Params())
   > Set GravityOverride(GravityOverride=@n3)
   Set VestAlpha(
    execute=@n2.then_1
    VestAlpha:double=float * float!(
     A=Get Is Sky Diving Vest Alpha() -> (Is Sky Diving Vest Alpha:double)
     B=Get Is Tactical Sprinting Alpha() -> (Is Tactical Sprinting Alpha:double)
     C:float=Get Is Sliding Alpha() -> (Is Sliding Alpha:float)
     D:double=Get Jumping Vest Alpha() -> (Jumping Vest Alpha:double)
     E:double=Get Is Swim Sprinting Alpha() -> (Is Swim Sprinting Alpha:double)
     F:double=Get Is Surface Swimming Alpha() -> (Is Surface Swimming Alpha:double)
     G:double=1
     H:float=Get Is Crouching Alpha() -> (Is Crouching Alpha:float)
    ) -> (ReturnValue) {is=title K2Node_PromotableOperator}
   ) -> (Output_Get:double)
   Is Valid(exec=@n2.then_10, InputObject=Get Player Pawn() -> (Player Pawn:Fort Player Pawn Athena Object Reference))
   > n4 = Branch(Condition=Get bIsInFrontEnd())
   Set WPO Globals Function(
    execute=@n4.else
    Material=Get Mat Instance Dynamic()
    Overlay Material=Get Overlay Mat Instance Dynamic()
    Skeletal Mesh=Get Skel Mesh()
    "Local Noise Transform  Location"=Get Local Noise Transform Location() -> (Local Noise Transform Location:Vector)
    Local Noise Offset=Get Local Noise Offset()
   )
   > Set WPO Regional Function!(
    execute
    Material:Material Instance Dynamic Object Reference=Get Mat Instance Dynamic()
    Overlay Material:Material Instance Dynamic Object Reference=Get Overlay Mat Instance Dynamic()
    Skeletal Mesh:Skeletal Mesh Component Object Reference=Get Skel Mesh()
    Velocity Transform:Name=spine_01
    Velocity Interp Speed:double=Get Velocity Interp Speed()
    Collision Transform:Name=spine_01
    Collision Transform Axis:EAxis Enum
    Invert Collision Transform Axis:bool=false
    Delta Time X:double=@n1.DeltaTimeX
    Previous Velocity Transform Location:"Vector (by ref)"=Get R1 Velocity Transform Location()
    Previous Velocity Transform Velocity:"Vector (by ref)"=Get R1 Velocity Transform Velocity()
    Region:int=1
   ) -> (then)
   > n5 = Branch(Condition=Get bIsSkyDiving())
   > n6 = Set WPO Pho Wind Intensity(WPO Pho Wind Intensity=1)
   SetScalarParameterValue(
    execute=[@n7.then, @n6.then]
    self=[Get Mat Instance Dynamic(), Get Overlay Mat Instance Dynamic()]
    ParameterName=ClothSim_WindStrength
    Value=Get WPO Pho Wind Intensity()
   )
   n7 = Set WPO Pho Wind Intensity(execute=@n5.else, WPO Pho Wind Intensity=0)
   n8 = Sequence!(execute=@n2.then_11) -> (then_0, then_1)
   n9 = CheckForBackbling()
   n10 = Set Is Wearing Backbling(execute=@n39.then_1, Is Wearing Backbling=OR Boolean(A=@n9, B=@n9.bIsBackblingLong))
   Set Backbling Physics Alpha(
    execute=@n8.then_1
    Backbling Physics Alpha:float=float * float!(A="To Float (Boolean)"(InBool=@n10), B=1, C:double=Get VestAlpha()) -> (ReturnValue)
   ) -> (Output_Get:float)
   Set Is Skydive Floating Alpha(execute=@n2.then_12, Is Skydive Floating Alpha:double="To Float (Boolean)"(InBool=Get bIsSkydiveFloating())) -> (Output_Get:double)
   Set Is Playing Melee Anim Float(execute=@n2.then_13, Is Playing Melee Anim Float:double="To Float (Boolean)"(InBool=Get bIsPlayingMeleeAnim())) -> (Output_Get:double)
   n11 = Sequence!(execute=@n2.then_2) -> (then_0, then_1, then_2:Exec, then_3:Exec, then_4:Exec, then_5:Exec, then_6:Exec)
   > n12 = Branch(Condition=Get bIsSkyDiving())
   > Set Is Sky Diving Vest Alpha(Is Sky Diving Vest Alpha:double=0.5) -> (Output_Get:double)
   Set Is Sky Diving Vest Alpha(execute=@n12.else, Is Sky Diving Vest Alpha=1)
   n13 = Get bIsCrouching()
   n14 = Get bIsSprinting()
   n15 = Get bIsTacticalSprinting()
   n16 = Branch(execute=@n11.then_1, Condition=AND Boolean(A=OR Boolean(A=@n15, B=@n14), B=NOT Boolean(A=@n13)))
   > Set Is Tactical Sprinting Alpha(Is Tactical Sprinting Alpha:double=0.25) -> (Output_Get:double)
   Set Is Tactical Sprinting Alpha(execute=@n16.else, Is Tactical Sprinting Alpha=1)
   n17 = Branch(execute=@n11.then_2, Condition=Get bIsSliding())
   > Set Is Sliding Alpha(Is Sliding Alpha:float=0.25) -> (Output_Get:float)
   Set Is Sliding Alpha(execute=@n17.else, Is Sliding Alpha=1)
   n18 = NOT Boolean(A=Get bIsSkyDiving())
   n19 = Get bIsJumping()
   n20 = NOT Boolean(A=Get bIsSkyDiving())
   n21 = Get bIsFalling()
   n22 = Branch(execute=@n11.then_3, Condition=OR Boolean(A=AND Boolean(A=@n19, B=@n18), B=AND Boolean(A=@n21, B=@n20)))
   > Set Jumping Vest Alpha(Jumping Vest Alpha:double=0.75) -> (Output_Get:double)
   Set Jumping Vest Alpha(execute=@n22.else, Jumping Vest Alpha=1)
   n23 = Branch(execute=@n11.then_4, Condition=Get bIsSurfaceSwimming())
   > Set Is Surface Swimming Alpha(Is Surface Swimming Alpha:double=1) -> (Output_Get:double)
   Set Is Surface Swimming Alpha(execute=@n23.else, Is Surface Swimming Alpha=1)
   n24 = Branch(execute=@n11.then_5, Condition=Get bIsSwimSprinting() -> (bIsSwimSprinting:bool))
   > Set Is Swim Sprinting Alpha(Is Swim Sprinting Alpha:double=1) -> (Output_Get:double)
   Set Is Swim Sprinting Alpha(execute=@n24.else, Is Swim Sprinting Alpha=1)
   n25 = Branch(execute=@n11.then_6, Condition=Get bIsCrouching())
   > Set Is Crouching Alpha(Is Crouching Alpha:float=1) -> (Output_Get:float)
   Set Is Crouching Alpha(execute=@n25.else, Is Crouching Alpha=1)
   Set Is Crouching Float(execute=@n2.then_4, Is Crouching Float:double="To Float (Boolean)"(InBool=Get bIsCrouching())) -> (Output_Get:double)
   n26 = Sequence!(execute=@n2.then_5) -> (then_0, then_1, then_2, then_3, then_4)
   > n27 = Cast To FortPlayerAnimInstance(Object=GetAnimInstance(self=GetOwnerSkeletalMeshComponent()))
   > Set isPistol(isPistol=false)
   > Set isRifle(isRifle:bool=false) -> (Output_Get:bool)
   > n28 = Switch on EFortWeaponCoreAnimation!(execute, Selection:EFortWeaponCoreAnimation Enum=Get WeaponCoreAnim(self=@n27)) -> (
    Melee:Exec
    Pistol:Exec
    Shotgun:Exec
    PaperBlueprint:Exec
    Rifle:Exec
    MeleeOneHand:Exec
    MachinePistol:Exec
    RocketLauncher:Exec
    GrenadeLauncher:Exec
    GoingCommando:Exec
    AssaultRifle:Exec
    TacticalShotgun:Exec
    SniperRifle:Exec
    TrapPlacement:Exec
    ShoulderLauncher:Exec
    AbilityDecoTool:Exec
    Crossbow:Exec
    C4:Exec
    RemoteControl:Exec
    DualWield:Exec
    AR_BullPup:Exec
    AR_ForwardGrip:Exec
    MedPackPaddles:Exec
    SMG_P90:Exec
    AR_DrumGun:Exec
    Consumable_Small:Exec
    Consumable_Large:Exec
    Balloon:Exec
    MountedTurret:Exec
    CreativeTool:Exec
    ExplosiveBow:Exec
    AshtonIndigo:Exec
    AshtonChicago:Exec
    MeleeDualWield:Exec
    Unarmed:Exec
   )
   Set isRifle(execute=[@n28.AssaultRifle, @n28.Rifle, @n28.RocketLauncher, @n28.Shotgun, @n28.SniperRifle, @n28.TacticalShotgun], isRifle=true)
   Set isPistol(execute=@n28.Pistol, isPistol=true)
   Set Is Rifle Alpha(execute=@n26.then_1, Is Rifle Alpha:double="To Float (Boolean)"(InBool=Get isRifle() -> (isRifle:bool))) -> (Output_Get:double)
   Set Is Pistol Alpha(execute=@n26.then_2, Is Pistol Alpha:double="To Float (Boolean)"(InBool=Get isPistol() -> (isPistol:bool))) -> (Output_Get:double)
   n29 = Get bIsTargeting()
   n30 = Get isRifle()
   Set Is Rifle Targeting Alpha(execute=@n26.then_3, Is Rifle Targeting Alpha:double="To Float (Boolean)"(InBool=AND Boolean(A=@n30, B=@n29))) -> (Output_Get:double)
   n31 = Get bIsTargeting()
   n32 = Get isPistol()
   Set Is Pistol Targeting Alpha(execute=@n26.then_4, Is Pistol Targeting Alpha:float="To Float (Boolean)"(InBool=AND Boolean(A=@n32, B=@n31))) -> (Output_Get:float)
   n33 = NOT Boolean(A=Get bIsSkyDiving())
   n34 = OR Boolean(A=Get bIsJumping(), B=Get bIsFalling())
   Set Is Jumping Float(execute=@n2.then_6, Is Jumping Float:double="To Float (Boolean)"(InBool=AND Boolean(A=@n34, B=@n33))) -> (Output_Get:double)
   Set Is Tactical Sprinting Float(execute=@n2.then_7, Is Tactical Sprinting Float:double="To Float (Boolean)"(InBool=Get bIsTacticalSprinting())) -> (Output_Get:double)
   n35 = Get bIsSwimSprinting()
   n36 = Get bIsSurfaceSwimming()
   Set Is swimming Float(execute=@n2.then_8, Is swimming Float:double="To Float (Boolean)"(InBool=OR Boolean(A=@n35, B=@n36))) -> (Output_Get:double)
   Set Is on Grind Rail Float(execute=@n2.then_9, Is on Grind Rail Float:double="To Float (Boolean)"(InBool=Get bIsOnGrindRail())) -> (Output_Get:double)
   Event BlueprintInitializeAnimation()
   > Set Skel Mesh(Skel Mesh=GetOwnerSkeletalMeshComponent())
   n37 = TryGetPawnOwner()
   > Is Valid(InputObject=@n37)
   > n38 = Cast To FortPlayerPawnAthena(Object=@n37)
   > Set Player Pawn(Player Pawn:Fort Player Pawn Athena Object Reference=@n38) -> (Output_Get:Fort Player Pawn Athena Object Reference)
   Event OnCharacterCustomizationLoaded()
   > n39 = Sequence!(execute) -> (then_0, then_1, then_2)
   n40 = Set Skel Mesh(execute=@n39.then_2, Skel Mesh=GetPartSkeletalMeshComponent())
   > n41 = Get Dynamic Material Instance(SkeletalMeshComponent=@n40)
   > Set Mat Instance Dynamic(Mat Instance Dynamic=@n41)
   > n42 = GetOrCreateDynamicOverlayMaterial(MeshComponent=@n40, MaterialName=None)
   > Set Overlay Mat Instance Dynamic(Overlay Mat Instance Dynamic=@n42)
   Set Backbling Physics Alpha(execute=@n43.then_1, Backbling Physics Alpha="To Float (Boolean)"(InBool=@n10))
   n43 = Sequence!(execute=@n10.then) -> (then_0, then_1)
  ]}
]
variables [
 Backbling Physics Alpha {type=number}
 Gravity Override Params {type=GravityOverrideParamsStruct}
 GravityOverride {type=Vector}
 Is Crouching Alpha {type=number}
 Is Crouching Float {type=number}
 Is Jumping Float {type=number}
 Is on Grind Rail Float {type=number}
 Is Pistol Alpha {type=number}
 Is Pistol Targeting Alpha {type=number}
 Is Playing Melee Anim Float {type=number}
 Is Rifle Alpha {type=number}
 Is Rifle Targeting Alpha {type=number}
 Is Sky Diving Vest Alpha {type=number}
 Is Skydive Floating Alpha {type=number}
 Is Sliding Alpha {type=number}
 Is Surface Swimming Alpha {type=number}
 Is Swim Sprinting Alpha {type=number}
 Is swimming Float {type=number}
 Is Tactical Sprinting Alpha {type=number}
 Is Tactical Sprinting Float {type=number}
 Is Wearing Backbling {type=boolean}
 isPistol {type=boolean}
 isRifle {type=boolean}
 Jumping Vest Alpha {type=number}
 Local Noise Offset {type=Vector}
 Local Noise Transform Location {type=Vector}
 Mat Instance Dynamic {type=MaterialInstanceDynamic}
 Overlay Mat Instance Dynamic {type=MaterialInstanceDynamic}
 Player Pawn {type=FortPlayerPawnAthena}
 R1 Velocity Transform Location {type=Vector}
 R1 Velocity Transform Velocity {type=Vector}
 Skel Mesh {type=SkeletalMeshComponent}
 Velocity Interp Speed {type=number, default=30}
 VestAlpha {type=number}
 WPO Pho Wind Intensity {type=number}
]
