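# Physics asset for the HydroIgnite male medium face accessory mesh, defining secondary motion for fourteen hair strands.
# Back, front, left and top strands are each a chain of two to five capsule joints, hung from a zero-radius MSTR root constrained to the head.
# Three head capsules approximate the skull and face as collision shapes.
# Tight swing and twist limits, narrowest in swing at the roots, keep the strands near their styled shape.
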PhysicsAsset /BRCosmetics/Characters/Player/Male/Medium/Bodies/M_MED_HydroIgnite/Meshes/Parts/M_MED_HydroIgnite_FaceAcc_Physics
bodies {
 dyn_bk_01hairstrand_01_jnt [{shapeType=Capsule, length=1, radius=1, shapeName=dyn_bk_01hairstrand_01_jnt_capsule, center=<1.5 0 0>, rotation=<rot 90 0 0>}]
 dyn_bk_01hairstrand_02_jnt [{shapeType=Capsule, length=1, radius=1, shapeName=dyn_bk_01hairstrand_02_jnt_capsule, center=<1.5 0 0>, rotation=<rot 90 0 0>}]
 dyn_bk_01hairstrand_03_jnt [{shapeType=Capsule, length=1, radius=1, shapeName=dyn_bk_01hairstrand_03_jnt_capsule, center=<1.5 0 0>, rotation=<rot 90 0 0>}]
 dyn_bk_01hairstrand_MSTR [{shapeType=Capsule, length=0.505, radius=0, shapeName=dyn_bk_01hairstrand_MSTR_capsule, rotation=<rot 0 -90 0>}]
 dyn_bk_02hairstrand_01_jnt [{shapeType=Capsule, length=1, radius=1, shapeName=dyn_bk_02hairstrand_01_jnt_capsule, center=<1.5 0 0>, rotation=<rot 90 0 0>}]
 dyn_bk_02hairstrand_02_jnt [{shapeType=Capsule, length=1, radius=1, shapeName=dyn_bk_02hairstrand_02_jnt_capsule, center=<1.5 0 0>, rotation=<rot 90 0 0>}]
 dyn_bk_02hairstrand_03_jnt [{shapeType=Capsule, length=1, radius=1, shapeName=dyn_bk_02hairstrand_03_jnt_capsule, center=<1.5 0 0>, rotation=<rot 90 0 0>}]
 dyn_bk_02hairstrand_MSTR [{shapeType=Capsule, length=0.505, radius=0, shapeName=dyn_bk_02hairstrand_MSTR_capsule, rotation=<rot 0 -90 0>}]
 dyn_bk_03hairstrand_01_jnt [{shapeType=Capsule, length=1, radius=1, shapeName=dyn_bk_03hairstrand_01_jnt_capsule, center=<1.5 0 0>, rotation=<rot 90 0 0>}]
 dyn_bk_03hairstrand_02_jnt [{shapeType=Capsule, length=1, radius=1, shapeName=dyn_bk_03hairstrand_02_jnt_capsule, center=<1.5 0 0>, rotation=<rot 90 0 0>}]
 dyn_bk_03hairstrand_03_jnt [{shapeType=Capsule, length=1, radius=1, shapeName=dyn_bk_03hairstrand_03_jnt_capsule, center=<1.5 0 0>, rotation=<rot 90 0 0>}]
 dyn_bk_03hairstrand_MSTR [{shapeType=Capsule, length=0.505, radius=0, shapeName=dyn_bk_03hairstrand_MSTR_capsule, rotation=<rot 0 -90 0>}]
 dyn_ft_01hairstrand_01_jnt [{shapeType=Capsule, length=1, radius=1, shapeName=dyn_ft_01hairstrand_01_jnt_capsule, center=<1.5 0 0>, rotation=<rot 90 0 0>}]
 dyn_ft_01hairstrand_02_jnt [{shapeType=Capsule, length=1, radius=1, shapeName=dyn_ft_01hairstrand_02_jnt_capsule, center=<1.5 0 0>, rotation=<rot 90 0 0>}]
 dyn_ft_01hairstrand_03_jnt [{shapeType=Capsule, length=1, radius=1, shapeName=dyn_ft_01hairstrand_03_jnt_capsule, center=<1.5 0 0>, rotation=<rot 90 0 0>}]
 dyn_ft_01hairstrand_04_jnt [{shapeType=Capsule, length=1, radius=1, shapeName=dyn_ft_01hairstrand_04_jnt_capsule, center=<1.5 0 0>, rotation=<rot 90 0 0>}]
 dyn_ft_01hairstrand_MSTR [{shapeType=Capsule, length=0.505, radius=0, shapeName=dyn_ft_01hairstrand_MSTR_capsule, rotation=<rot 0 -90 0>}]
 dyn_ft_02hairstrand_01_jnt [{shapeType=Capsule, length=1, radius=1, shapeName=dyn_ft_02hairstrand_01_jnt_capsule, center=<1.5 0 0>, rotation=<rot 90 0 0>}]
 dyn_ft_02hairstrand_02_jnt [{shapeType=Capsule, length=1, radius=1, shapeName=dyn_ft_02hairstrand_02_jnt_capsule, center=<1.5 0 0>, rotation=<rot 90 0 0>}]
 dyn_ft_02hairstrand_03_jnt [{shapeType=Capsule, length=1, radius=1, shapeName=dyn_ft_02hairstrand_03_jnt_capsule, center=<1.5 0 0>, rotation=<rot 90 0 0>}]
 dyn_ft_02hairstrand_04_jnt [{shapeType=Capsule, length=1, radius=1, shapeName=dyn_ft_02hairstrand_04_jnt_capsule, center=<1.5 0 0>, rotation=<rot 90 0 0>}]
 dyn_ft_02hairstrand_05_jnt [{shapeType=Capsule, length=1, radius=1, shapeName=dyn_ft_02hairstrand_05_jnt_capsule, center=<1.5 0 0>, rotation=<rot 90 0 0>}]
 dyn_ft_02hairstrand_MSTR [{shapeType=Capsule, length=0.505, radius=0, shapeName=dyn_ft_02hairstrand_MSTR_capsule, rotation=<rot 0 -90 0>}]
 dyn_ft_03hairstrand_01_jnt [{shapeType=Capsule, length=1, radius=1, shapeName=dyn_ft_03hairstrand_01_jnt_capsule, center=<1.5 0 0>, rotation=<rot 90 0 0>}]
 dyn_ft_03hairstrand_02_jnt [{shapeType=Capsule, length=1, radius=1, shapeName=dyn_ft_03hairstrand_02_jnt_capsule, center=<1.5 0 0>, rotation=<rot 90 0 0>}]
 dyn_ft_03hairstrand_03_jnt [{shapeType=Capsule, length=1, radius=1, shapeName=dyn_ft_03hairstrand_03_jnt_capsule, center=<1.5 0 0>, rotation=<rot 90 0 0>}]
 dyn_ft_03hairstrand_MSTR [{shapeType=Capsule, length=0.505, radius=0, shapeName=dyn_ft_03hairstrand_MSTR_capsule, rotation=<rot 0 -90 0>}]
 dyn_lf_01hairstrand_01_jnt [{shapeType=Capsule, length=1, radius=1, shapeName=dyn_lf_01hairstrand_01_jnt_capsule, center=<1.5 0 0>, rotation=<rot 90 0 0>}]
 dyn_lf_01hairstrand_02_jnt [{shapeType=Capsule, length=1, radius=1, shapeName=dyn_lf_01hairstrand_02_jnt_capsule, center=<1.5 0 0>, rotation=<rot 90 0 0>}]
 dyn_lf_01hairstrand_MSTR [{shapeType=Capsule, length=0.505, radius=0, shapeName=dyn_lf_01hairstrand_MSTR_capsule, rotation=<rot 0 -90 0>}]
 dyn_lf_02hairstrand_01_jnt [{shapeType=Capsule, length=1, radius=1, shapeName=dyn_lf_02hairstrand_01_jnt_capsule, center=<1.5 0 0>, rotation=<rot 90 0 0>}]
 dyn_lf_02hairstrand_02_jnt [{shapeType=Capsule, length=1, radius=1, shapeName=dyn_lf_02hairstrand_02_jnt_capsule, center=<1.5 0 0>, rotation=<rot 90 0 0>}]
 dyn_lf_02hairstrand_MSTR [{shapeType=Capsule, length=0.505, radius=0, shapeName=dyn_lf_02hairstrand_MSTR_capsule, rotation=<rot 0 -90 0>}]
 dyn_lf_03hairstrand_01_jnt [{shapeType=Capsule, length=1, radius=1, shapeName=dyn_lf_03hairstrand_01_jnt_capsule, center=<1.5 0 0>, rotation=<rot 90 0 0>}]
 dyn_lf_03hairstrand_02_jnt [{shapeType=Capsule, length=1, radius=1, shapeName=dyn_lf_03hairstrand_02_jnt_capsule, center=<1.5 0 0>, rotation=<rot 90 0 0>}]
 dyn_lf_03hairstrand_03_jnt [{shapeType=Capsule, length=1, radius=1, shapeName=dyn_lf_03hairstrand_03_jnt_capsule, center=<1.5 0 0>, rotation=<rot 90 0 0>}]
 dyn_lf_03hairstrand_MSTR [{shapeType=Capsule, length=0.505, radius=0, shapeName=dyn_lf_03hairstrand_MSTR_capsule, rotation=<rot 0 -90 0>}]
 dyn_lf_04hairstrand_01_jnt [{shapeType=Capsule, length=1, radius=1, shapeName=dyn_lf_04hairstrand_01_jnt_capsule, center=<1.5 0 0>, rotation=<rot 90 0 0>}]
 dyn_lf_04hairstrand_02_jnt [{shapeType=Capsule, length=1, radius=1, shapeName=dyn_lf_04hairstrand_02_jnt_capsule, center=<1.5 0 0>, rotation=<rot 90 0 0>}]
 dyn_lf_04hairstrand_03_jnt [{shapeType=Capsule, length=1, radius=1, shapeName=dyn_lf_04hairstrand_03_jnt_capsule, center=<1.5 0 0>, rotation=<rot 90 0 0>}]
 dyn_lf_04hairstrand_MSTR [{shapeType=Capsule, length=0.505, radius=0, shapeName=dyn_lf_04hairstrand_MSTR_capsule, rotation=<rot 0 -90 0>}]
 dyn_tp_01hairstrand_01_jnt [{shapeType=Capsule, length=1, radius=1, shapeName=dyn_tp_01hairstrand_01_jnt_capsule, center=<1.5 0 0>, rotation=<rot 90 0 0>}]
 dyn_tp_01hairstrand_02_jnt [{shapeType=Capsule, length=1, radius=1, shapeName=dyn_tp_01hairstrand_02_jnt_capsule, center=<1.5 0 0>, rotation=<rot 90 0 0>}]
 dyn_tp_01hairstrand_03_jnt [{shapeType=Capsule, length=1, radius=1, shapeName=dyn_tp_01hairstrand_03_jnt_capsule, center=<1.5 0 0>, rotation=<rot 90 0 0>}]
 dyn_tp_01hairstrand_MSTR [{shapeType=Capsule, length=0.505, radius=0, shapeName=dyn_tp_01hairstrand_MSTR_capsule, rotation=<rot 0 -90 0>}]
 dyn_tp_02hairstrand_01_jnt [{shapeType=Capsule, length=1, radius=1, shapeName=dyn_tp_02hairstrand_01_jnt_capsule, center=<1.5 0 0>, rotation=<rot 90 0 0>}]
 dyn_tp_02hairstrand_02_jnt [{shapeType=Capsule, length=1, radius=1, shapeName=dyn_tp_02hairstrand_02_jnt_capsule, center=<1.5 0 0>, rotation=<rot 90 0 0>}]
 dyn_tp_02hairstrand_03_jnt [{shapeType=Capsule, length=1, radius=1, shapeName=dyn_tp_02hairstrand_03_jnt_capsule, center=<1.5 0 0>, rotation=<rot 90 0 0>}]
 dyn_tp_02hairstrand_MSTR [{shapeType=Capsule, length=0.505, radius=0, shapeName=dyn_tp_02hairstrand_MSTR_capsule, rotation=<rot 0 -90 0>}]
 dyn_tp_03hairstrand_01_jnt [{shapeType=Capsule, length=1, radius=1, shapeName=dyn_tp_03hairstrand_01_jnt_capsule, center=<1.5 0 0>, rotation=<rot 90 0 0>}]
 dyn_tp_03hairstrand_02_jnt [{shapeType=Capsule, length=1, radius=1, shapeName=dyn_tp_03hairstrand_02_jnt_capsule, center=<1.5 0 0>, rotation=<rot 90 0 0>}]
 dyn_tp_03hairstrand_03_jnt [{shapeType=Capsule, length=1, radius=1, shapeName=dyn_tp_03hairstrand_03_jnt_capsule, center=<1.5 0 0>, rotation=<rot 90 0 0>}]
 dyn_tp_03hairstrand_MSTR [{shapeType=Capsule, length=0.505, radius=0, shapeName=dyn_tp_03hairstrand_MSTR_capsule, rotation=<rot 0 -90 0>}]
 dyn_tp_04hairstrand_01_jnt [{shapeType=Capsule, length=1, radius=1, shapeName=dyn_tp_04hairstrand_01_jnt_capsule, center=<1.5 0 0>, rotation=<rot 90 0 0>}]
 dyn_tp_04hairstrand_02_jnt [{shapeType=Capsule, length=1, radius=1, shapeName=dyn_tp_04hairstrand_02_jnt_capsule, center=<1.5 0 0>, rotation=<rot 90 0 0>}]
 dyn_tp_04hairstrand_03_jnt [{shapeType=Capsule, length=1, radius=1, shapeName=dyn_tp_04hairstrand_03_jnt_capsule, center=<1.5 0 0>, rotation=<rot 90 0 0>}]
 dyn_tp_04hairstrand_MSTR [{shapeType=Capsule, length=0.505, radius=0, shapeName=dyn_tp_04hairstrand_MSTR_capsule, rotation=<rot 0 -90 0>}]
 head [
  {shapeType=Capsule, length=6, radius=6, shapeName=head_capsule, center=<3.5 3 0>, rotation=<rot 90 0 0>}
  {shapeType=Capsule, length=3, radius=4, shapeName=head_capsule, center=<3.6424 6.3431 0>, rotation=<rot 80 -90 -90>}
  {shapeType=Capsule, length=3, radius=0.75, shapeName=head_capsule, center=<5.5988 8.8015 2.8054>, rotation=<rot 0 -0.0002 -30>}
 ]
}
constraints [
 {bone1Name=dyn_bk_01hairstrand_MSTR, bone2Name=head, swing1LimitDegrees=5, swing2LimitDegrees=5, twistLimitDegrees=2}
 {bone1Name=dyn_bk_01hairstrand_01_jnt, bone2Name=dyn_bk_01hairstrand_MSTR, swing1LimitDegrees=10, swing2LimitDegrees=10, twistLimitDegrees=2}
 {
  bone1Name=dyn_bk_01hairstrand_02_jnt
  bone2Name=dyn_bk_01hairstrand_01_jnt
  swing1LimitDegrees=10
  swing2LimitDegrees=10
  twistLimitDegrees=2
 }
 {
  bone1Name=dyn_bk_01hairstrand_03_jnt
  bone2Name=dyn_bk_01hairstrand_02_jnt
  swing1LimitDegrees=10
  swing2LimitDegrees=10
  twistLimitDegrees=2
 }
 {bone1Name=dyn_bk_02hairstrand_MSTR, bone2Name=head, swing1LimitDegrees=5, swing2LimitDegrees=5, twistLimitDegrees=2}
 {bone1Name=dyn_bk_02hairstrand_01_jnt, bone2Name=dyn_bk_02hairstrand_MSTR, swing1LimitDegrees=10, swing2LimitDegrees=10, twistLimitDegrees=2}
 {
  bone1Name=dyn_bk_02hairstrand_02_jnt
  bone2Name=dyn_bk_02hairstrand_01_jnt
  swing1LimitDegrees=10
  swing2LimitDegrees=10
  twistLimitDegrees=2
 }
 {
  bone1Name=dyn_bk_02hairstrand_03_jnt
  bone2Name=dyn_bk_02hairstrand_02_jnt
  swing1LimitDegrees=10
  swing2LimitDegrees=10
  twistLimitDegrees=2
 }
 {bone1Name=dyn_bk_03hairstrand_MSTR, bone2Name=head, swing1LimitDegrees=5, swing2LimitDegrees=5, twistLimitDegrees=2}
 {bone1Name=dyn_bk_03hairstrand_01_jnt, bone2Name=dyn_bk_03hairstrand_MSTR, swing1LimitDegrees=10, swing2LimitDegrees=10, twistLimitDegrees=2}
 {
  bone1Name=dyn_bk_03hairstrand_02_jnt
  bone2Name=dyn_bk_03hairstrand_01_jnt
  swing1LimitDegrees=10
  swing2LimitDegrees=10
  twistLimitDegrees=2
 }
 {
  bone1Name=dyn_bk_03hairstrand_03_jnt
  bone2Name=dyn_bk_03hairstrand_02_jnt
  swing1LimitDegrees=10
  swing2LimitDegrees=10
  twistLimitDegrees=2
 }
 {bone1Name=dyn_ft_01hairstrand_MSTR, bone2Name=head, swing1LimitDegrees=5, swing2LimitDegrees=5, twistLimitDegrees=2}
 {bone1Name=dyn_ft_01hairstrand_01_jnt, bone2Name=dyn_ft_01hairstrand_MSTR, swing1LimitDegrees=10, swing2LimitDegrees=10, twistLimitDegrees=2}
 {
  bone1Name=dyn_ft_01hairstrand_02_jnt
  bone2Name=dyn_ft_01hairstrand_01_jnt
  swing1LimitDegrees=10
  swing2LimitDegrees=10
  twistLimitDegrees=2
 }
 {
  bone1Name=dyn_ft_01hairstrand_03_jnt
  bone2Name=dyn_ft_01hairstrand_02_jnt
  swing1LimitDegrees=10
  swing2LimitDegrees=10
  twistLimitDegrees=2
 }
 {
  bone1Name=dyn_ft_01hairstrand_04_jnt
  bone2Name=dyn_ft_01hairstrand_03_jnt
  swing1LimitDegrees=10
  swing2LimitDegrees=10
  twistLimitDegrees=2
 }
 {bone1Name=dyn_ft_02hairstrand_MSTR, bone2Name=head, swing1LimitDegrees=5, swing2LimitDegrees=5, twistLimitDegrees=2}
 {bone1Name=dyn_ft_02hairstrand_01_jnt, bone2Name=dyn_ft_02hairstrand_MSTR, swing1LimitDegrees=10, swing2LimitDegrees=10, twistLimitDegrees=2}
 {
  bone1Name=dyn_ft_02hairstrand_02_jnt
  bone2Name=dyn_ft_02hairstrand_01_jnt
  swing1LimitDegrees=10
  swing2LimitDegrees=10
  twistLimitDegrees=2
 }
 {
  bone1Name=dyn_ft_02hairstrand_03_jnt
  bone2Name=dyn_ft_02hairstrand_02_jnt
  swing1LimitDegrees=10
  swing2LimitDegrees=10
  twistLimitDegrees=2
 }
 {
  bone1Name=dyn_ft_02hairstrand_04_jnt
  bone2Name=dyn_ft_02hairstrand_03_jnt
  swing1LimitDegrees=10
  swing2LimitDegrees=10
  twistLimitDegrees=2
 }
 {
  bone1Name=dyn_ft_02hairstrand_05_jnt
  bone2Name=dyn_ft_02hairstrand_04_jnt
  swing1LimitDegrees=10
  swing2LimitDegrees=10
  twistLimitDegrees=2
 }
 {bone1Name=dyn_ft_03hairstrand_MSTR, bone2Name=head, swing1LimitDegrees=5, swing2LimitDegrees=5, twistLimitDegrees=2}
 {bone1Name=dyn_ft_03hairstrand_01_jnt, bone2Name=dyn_ft_03hairstrand_MSTR, swing1LimitDegrees=10, swing2LimitDegrees=10, twistLimitDegrees=2}
 {
  bone1Name=dyn_ft_03hairstrand_02_jnt
  bone2Name=dyn_ft_03hairstrand_01_jnt
  swing1LimitDegrees=10
  swing2LimitDegrees=10
  twistLimitDegrees=2
 }
 {
  bone1Name=dyn_ft_03hairstrand_03_jnt
  bone2Name=dyn_ft_03hairstrand_02_jnt
  swing1LimitDegrees=10
  swing2LimitDegrees=10
  twistLimitDegrees=2
 }
 {bone1Name=dyn_lf_01hairstrand_MSTR, bone2Name=head, swing1LimitDegrees=5, swing2LimitDegrees=5, twistLimitDegrees=2}
 {bone1Name=dyn_lf_01hairstrand_01_jnt, bone2Name=dyn_lf_01hairstrand_MSTR, swing1LimitDegrees=10, swing2LimitDegrees=10, twistLimitDegrees=2}
 {
  bone1Name=dyn_lf_01hairstrand_02_jnt
  bone2Name=dyn_lf_01hairstrand_01_jnt
  swing1LimitDegrees=10
  swing2LimitDegrees=10
  twistLimitDegrees=2
 }
 {bone1Name=dyn_lf_02hairstrand_MSTR, bone2Name=head, swing1LimitDegrees=5, swing2LimitDegrees=5, twistLimitDegrees=2}
 {bone1Name=dyn_lf_02hairstrand_01_jnt, bone2Name=dyn_lf_02hairstrand_MSTR, swing1LimitDegrees=10, swing2LimitDegrees=10, twistLimitDegrees=2}
 {
  bone1Name=dyn_lf_02hairstrand_02_jnt
  bone2Name=dyn_lf_02hairstrand_01_jnt
  swing1LimitDegrees=10
  swing2LimitDegrees=10
  twistLimitDegrees=2
 }
 {bone1Name=dyn_lf_03hairstrand_MSTR, bone2Name=head, swing1LimitDegrees=5, swing2LimitDegrees=5, twistLimitDegrees=2}
 {bone1Name=dyn_lf_03hairstrand_01_jnt, bone2Name=dyn_lf_03hairstrand_MSTR, swing1LimitDegrees=10, swing2LimitDegrees=10, twistLimitDegrees=2}
 {
  bone1Name=dyn_lf_03hairstrand_02_jnt
  bone2Name=dyn_lf_03hairstrand_01_jnt
  swing1LimitDegrees=10
  swing2LimitDegrees=10
  twistLimitDegrees=2
 }
 {
  bone1Name=dyn_lf_03hairstrand_03_jnt
  bone2Name=dyn_lf_03hairstrand_02_jnt
  swing1LimitDegrees=10
  swing2LimitDegrees=10
  twistLimitDegrees=2
 }
 {bone1Name=dyn_lf_04hairstrand_MSTR, bone2Name=head, swing1LimitDegrees=5, swing2LimitDegrees=5, twistLimitDegrees=2}
 {bone1Name=dyn_lf_04hairstrand_01_jnt, bone2Name=dyn_lf_04hairstrand_MSTR, swing1LimitDegrees=10, swing2LimitDegrees=10, twistLimitDegrees=2}
 {
  bone1Name=dyn_lf_04hairstrand_02_jnt
  bone2Name=dyn_lf_04hairstrand_01_jnt
  swing1LimitDegrees=10
  swing2LimitDegrees=10
  twistLimitDegrees=2
 }
 {
  bone1Name=dyn_lf_04hairstrand_03_jnt
  bone2Name=dyn_lf_04hairstrand_02_jnt
  swing1LimitDegrees=10
  swing2LimitDegrees=10
  twistLimitDegrees=2
 }
 {bone1Name=dyn_tp_01hairstrand_MSTR, bone2Name=head, swing1LimitDegrees=5, swing2LimitDegrees=5, twistLimitDegrees=2}
 {bone1Name=dyn_tp_01hairstrand_01_jnt, bone2Name=dyn_tp_01hairstrand_MSTR, swing1LimitDegrees=10, swing2LimitDegrees=10, twistLimitDegrees=2}
 {
  bone1Name=dyn_tp_01hairstrand_02_jnt
  bone2Name=dyn_tp_01hairstrand_01_jnt
  swing1LimitDegrees=10
  swing2LimitDegrees=10
  twistLimitDegrees=2
 }
 {
  bone1Name=dyn_tp_01hairstrand_03_jnt
  bone2Name=dyn_tp_01hairstrand_02_jnt
  swing1LimitDegrees=10
  swing2LimitDegrees=10
  twistLimitDegrees=2
 }
 {bone1Name=dyn_tp_02hairstrand_MSTR, bone2Name=head, swing1LimitDegrees=5, swing2LimitDegrees=5, twistLimitDegrees=2}
 {bone1Name=dyn_tp_02hairstrand_01_jnt, bone2Name=dyn_tp_02hairstrand_MSTR, swing1LimitDegrees=10, swing2LimitDegrees=10, twistLimitDegrees=2}
 {
  bone1Name=dyn_tp_02hairstrand_02_jnt
  bone2Name=dyn_tp_02hairstrand_01_jnt
  swing1LimitDegrees=10
  swing2LimitDegrees=10
  twistLimitDegrees=2
 }
 {
  bone1Name=dyn_tp_02hairstrand_03_jnt
  bone2Name=dyn_tp_02hairstrand_02_jnt
  swing1LimitDegrees=10
  swing2LimitDegrees=10
  twistLimitDegrees=2
 }
 {bone1Name=dyn_tp_03hairstrand_MSTR, bone2Name=head, swing1LimitDegrees=5, swing2LimitDegrees=5, twistLimitDegrees=2}
 {bone1Name=dyn_tp_03hairstrand_01_jnt, bone2Name=dyn_tp_03hairstrand_MSTR, swing1LimitDegrees=10, swing2LimitDegrees=10, twistLimitDegrees=2}
 {
  bone1Name=dyn_tp_03hairstrand_02_jnt
  bone2Name=dyn_tp_03hairstrand_01_jnt
  swing1LimitDegrees=10
  swing2LimitDegrees=10
  twistLimitDegrees=2
 }
 {
  bone1Name=dyn_tp_03hairstrand_03_jnt
  bone2Name=dyn_tp_03hairstrand_02_jnt
  swing1LimitDegrees=10
  swing2LimitDegrees=10
  twistLimitDegrees=2
 }
 {bone1Name=dyn_tp_04hairstrand_MSTR, bone2Name=head, swing1LimitDegrees=5, swing2LimitDegrees=5, twistLimitDegrees=2}
 {bone1Name=dyn_tp_04hairstrand_01_jnt, bone2Name=dyn_tp_04hairstrand_MSTR, swing1LimitDegrees=10, swing2LimitDegrees=10, twistLimitDegrees=2}
 {
  bone1Name=dyn_tp_04hairstrand_02_jnt
  bone2Name=dyn_tp_04hairstrand_01_jnt
  swing1LimitDegrees=10
  swing2LimitDegrees=10
  twistLimitDegrees=2
 }
 {
  bone1Name=dyn_tp_04hairstrand_03_jnt
  bone2Name=dyn_tp_04hairstrand_02_jnt
  swing1LimitDegrees=10
  swing2LimitDegrees=10
  twistLimitDegrees=2
 }
]
